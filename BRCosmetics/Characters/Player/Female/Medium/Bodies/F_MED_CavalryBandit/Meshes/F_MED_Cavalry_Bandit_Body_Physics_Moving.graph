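# Physics asset for the female medium Cavalry Bandit body cosmetic that sets up jiggle on its pouches, canisters, back straps and zippers.
# Capsules on pelvis, thighs and calves form the body, with one box on the back zipper.
# Tiny MSTR anchor bodies link most accessories to pelvis, thighs or parent pouches, mostly with free rotation, while angle-limited constraints shape the swing.
# Straps and side pouches form chains whose swing limits mostly widen toward the tips.
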PhysicsAsset /BRCosmetics/Characters/Player/Female/Medium/Bodies/F_MED_CavalryBandit/Meshes/F_MED_Cavalry_Bandit_Body_Physics_Moving
bodies {
 calf_l [{shapeType=Capsule, length=24.1857, radius=6.1289, center=<-12.8418 -0.0004 0>, rotation=<rot 90 0.0099 0.0099>}]
 calf_r [{shapeType=Capsule, length=24.1857, radius=6.6572, center=<14.2257 0.0004 0>, rotation=<rot 90 -0.0001 179.9999>}]
 dyn_cannister_01_l [{shapeType=Capsule, length=10, radius=3.5, shapeName=dyn_cannister_01_l_capsule, center=<2.1931 0 0>}]
 dyn_cannister_01_l_MSTR [{shapeType=Capsule, length=0.01, radius=0.01, rotation=<rot 0 -90 0>}]
 dyn_cannister_01_r [{shapeType=Capsule, length=10, radius=3.5, shapeName=dyn_cannister_01_r_capsule, center=<-2.3333 0 0>}]
 dyn_cannister_01_r_MSTR [{shapeType=Capsule, length=0.01, radius=0.01, rotation=<rot 0 -90 0>}]
 dyn_pouch_01_l [{
   shapeType=Capsule
   length=6.9604
   radius=1.6281
   shapeName=dyn_pouch_01_l_capsule
   center=<0.6888 0.4549 0.1658>
   rotation=<rot 37.211 3.3275 -82.418>
  }]
 dyn_pouch_01_l_MSTR [{shapeType=Capsule, length=0.01, radius=0.01, rotation=<rot 0 -90 0>}]
 dyn_pouch_01_r [{shapeType=Capsule, length=5.6812, radius=8.7, shapeName=dyn_pouch_01_r_capsule, center=<-18.011 0 0>, rotation=<rot 0 0 -90>}]
 dyn_pouch_01_r_MSTR [{shapeType=Capsule, length=0.01, radius=0.01, rotation=<rot 0 -90 0>}]
 dyn_pouch_02_l [{shapeType=Capsule, length=5.5544, radius=8.1876, center=<1.7741 1.9571 1.702>, rotation=<rot 32.6691 9.0757 -85.0432>}]
 dyn_pouch_02_r [{shapeType=Capsule, length=0.01, radius=0.01}]
 dyn_pouch_03_l [{shapeType=Capsule, length=3.552, radius=3.0117, shapeName=dyn_pouch_03_l_capsule, center=<4.6379 0 0>, rotation=<rot 90 0 0>}]
 dyn_pouch_03_r [{shapeType=Capsule, length=3.7957, radius=3.1165, shapeName=dyn_pouch_03_r_capsule, center=<-4.7172 0 0>, rotation=<rot 90 0 0>}]
 dyn_pouch_04_l [{shapeType=Capsule, length=6.7892, radius=2.6742, shapeName=dyn_pouch_04_l_capsule, center=<7.2939 0 0>, rotation=<rot 90 0 0>}]
 dyn_pouch_04_r [{shapeType=Capsule, length=6.5061, radius=3.0426, shapeName=dyn_pouch_04_r_capsule, center=<-6.6894 0 -0.0253>, rotation=<rot 90 0 0>}]
 dyn_pouch_bk_01 [{
   shapeType=Capsule
   length=8.5826
   radius=5.1391
   shapeName=dyn_pouch_bk_01_capsule
   center=<0.5347 -0.1652 2.9459>
   rotation=<rot 0 -3.3713 -89.5952>
  }]
 dyn_pouch_bk_01_MSTR [{shapeType=Capsule, length=0.01, radius=0.01, rotation=<rot 0 -90 0>}]
 dyn_pouch_bk_01_r [{
   shapeType=Capsule
   length=15
   radius=2
   shapeName=dyn_pouch_bk_01_r_capsule
   center=<-7.5126 1.5698 0.1885>
   rotation=<rot 88.563 0.0001 0.0001>
  }]
 dyn_pouch_fr_01 [{
   shapeType=Capsule
   length=6.3573
   radius=4.2186
   shapeName=dyn_pouch_fr_01_capsule
   center=<1.1691 0.2997 0.1506>
   rotation=<rot -0.0001 -179.4603 -91.8872>
  }]
 dyn_pouch_fr_01_MSTR [{shapeType=Capsule, length=0.01, radius=0.01, rotation=<rot 0 -90 0>}]
 dyn_pouch_fr_01_l [{shapeType=Capsule, length=19, radius=4.8626, center=<9.44 2.3844 -0.0003>, rotation=<rot 90 0 0>}]
 dyn_pouch_fr_01_l_MSTR [{shapeType=Capsule, length=0.01, radius=0.01, rotation=<rot 0 -90 0>}]
 dyn_pouch_fr_01_r [{shapeType=Capsule, length=19, radius=4.7695, center=<-9.44 -2.871 0.0002>, rotation=<rot 90 0 0>}]
 dyn_pouch_fr_01_r_MSTR [{shapeType=Capsule, length=0.01, radius=0.01, rotation=<rot 0 -90 0>}]
 dyn_strap_bk_01_l [{shapeType=Capsule, length=4, radius=3, shapeName=dyn_strap_bk_01_l_capsule, center=<4.8 0 0.5518>, rotation=<rot 90 0 0>}]
 dyn_strap_bk_01_r [{shapeType=Capsule, length=4, radius=3, shapeName=dyn_strap_bk_01_r_capsule, center=<4.8 0 0.5518>, rotation=<rot 90 0 0>}]
 dyn_strap_bk_02_l [{shapeType=Capsule, length=4, radius=3, shapeName=dyn_strap_bk_02_l_capsule, center=<4.8 0 0.5518>, rotation=<rot 90 0 0>}]
 dyn_strap_bk_02_r [{shapeType=Capsule, length=4, radius=3, shapeName=dyn_strap_bk_02_r_capsule, center=<4.8 0 0.5518>, rotation=<rot 90 0 0>}]
 dyn_strap_bk_03_l [{shapeType=Capsule, length=4, radius=3, shapeName=dyn_strap_bk_03_l_capsule, center=<4.8 0 0.5518>, rotation=<rot 90 0 0>}]
 dyn_strap_bk_03_r [{shapeType=Capsule, length=4, radius=3, shapeName=dyn_strap_bk_03_r_capsule, center=<4.8 0 0.5518>, rotation=<rot 90 0 0>}]
 dyn_strap_bk_04_l [{shapeType=Capsule, length=4, radius=3, shapeName=dyn_strap_bk_04_l_capsule, center=<4.8 0 0.5518>, rotation=<rot 90 0 0>}]
 dyn_strap_bk_04_r [{shapeType=Capsule, length=4, radius=3, shapeName=dyn_strap_bk_04_r_capsule, center=<4.8 0 0.5518>, rotation=<rot 90 0 0>}]
 dyn_zipper_01 [{shapeType=Capsule, length=2.5, radius=0.5, shapeName=dyn_zipper_01_capsule, center=<1.2531 -0.0001 -0.0019>, rotation=<rot 90 0 0>}]
 dyn_zipper_01_MSTR [{shapeType=Capsule, length=0.01, radius=0.01, rotation=<rot 0 -90 0>}]
 dyn_zipper_bk_01 [{shapeType=Box, extentX=1.01, extentY=0.5, extentZ=2.01, center=<1.2531 -0.0001 -0.0018>, rotation=<rot 0 -90.0028 -89.9365>}]
 dyn_zipper_bk_01_MSTR [{shapeType=Capsule, length=0.01, radius=0.01, rotation=<rot 0 -90 0>}]
 pelvis [{shapeType=Capsule, length=12.4909, radius=9.8849, center=<-7.5675 -1.1273 0>}]
 thigh_l [{shapeType=Capsule, length=41.1857, radius=6.6289, center=<-22.354 0.0004 0>, rotation=<rot 90 -0.0049 -0.0049>}]
 thigh_r [{shapeType=Capsule, length=41.1857, radius=6.6289, center=<22.354 -0.0004 0>, rotation=<rot 90 0.0297 -179.9703>}]
}
constraints [
 {
  bone1Name=dyn_zipper_01_MSTR
  bone2Name=dyn_pouch_fr_01
  swing1LimitDegrees=45
  swing1Motion=Free
  swing2LimitDegrees=45
  swing2Motion=Free
  twistLimitDegrees=45
  twistMotion=Free
 }
 {
  bone1Name=dyn_zipper_bk_01_MSTR
  bone2Name=dyn_pouch_bk_01
  swing1LimitDegrees=45
  swing1Motion=Free
  swing2LimitDegrees=45
  swing2Motion=Free
  twistLimitDegrees=45
  twistMotion=Free
 }
 {bone1Name=dyn_pouch_fr_01, bone2Name=dyn_pouch_fr_01_MSTR, swing1LimitDegrees=25, swing2LimitDegrees=25, twistLimitDegrees=5}
 {bone1Name=dyn_zipper_01, bone2Name=dyn_zipper_01_MSTR, swing1LimitDegrees=45, swing2LimitDegrees=45, twistLimitDegrees=5}
 {bone1Name=dyn_pouch_bk_01, bone2Name=dyn_pouch_bk_01_MSTR, swing1LimitDegrees=0.1, swing2LimitDegrees=5, twistLimitDegrees=0.1}
 {bone1Name=dyn_strap_bk_02_l, bone2Name=dyn_strap_bk_01_l, swing1LimitDegrees=25, swing2LimitDegrees=25, twistLimitDegrees=5}
 {bone1Name=dyn_strap_bk_03_l, bone2Name=dyn_strap_bk_02_l, swing1LimitDegrees=45, swing2LimitDegrees=45, twistLimitDegrees=5}
 {bone1Name=dyn_strap_bk_04_l, bone2Name=dyn_strap_bk_03_l, swing1LimitDegrees=75, swing2LimitDegrees=75, twistLimitDegrees=5}
 {bone1Name=dyn_zipper_bk_01, bone2Name=dyn_zipper_bk_01_MSTR, swing1LimitDegrees=65, swing2LimitDegrees=65, twistLimitDegrees=5}
 {bone1Name=dyn_cannister_01_l, bone2Name=dyn_cannister_01_l_MSTR, swing1LimitDegrees=45, swing2LimitDegrees=45, twistLimitDegrees=5}
 {
  bone1Name=dyn_pouch_fr_01_MSTR
  bone2Name=pelvis
  swing1LimitDegrees=45
  swing1Motion=Free
  swing2LimitDegrees=45
  swing2Motion=Free
  twistLimitDegrees=45
  twistMotion=Free
 }
 {
  bone1Name=dyn_cannister_01_l_MSTR
  bone2Name=pelvis
  swing1LimitDegrees=45
  swing1Motion=Free
  swing2LimitDegrees=45
  swing2Motion=Free
  twistLimitDegrees=45
  twistMotion=Free
 }
 {
  bone1Name=dyn_cannister_01_r_MSTR
  bone2Name=pelvis
  swing1LimitDegrees=45
  swing1Motion=Free
  swing2LimitDegrees=45
  swing2Motion=Free
  twistLimitDegrees=45
  twistMotion=Free
 }
 {
  bone1Name=dyn_pouch_fr_01_l_MSTR
  bone2Name=thigh_l
  swing1LimitDegrees=45
  swing1Motion=Free
  swing2LimitDegrees=45
  swing2Motion=Free
  twistLimitDegrees=45
  twistMotion=Free
 }
 {
  bone1Name=dyn_pouch_01_r_MSTR
  bone2Name=thigh_r
  swing1LimitDegrees=45
  swing1Motion=Free
  swing2LimitDegrees=45
  swing2Motion=Free
  twistLimitDegrees=45
  twistMotion=Free
 }
 {
  bone1Name=dyn_pouch_fr_01_r_MSTR
  bone2Name=thigh_r
  swing1LimitDegrees=45
  swing1Motion=Free
  swing2LimitDegrees=45
  swing2Motion=Free
  twistLimitDegrees=45
  twistMotion=Free
 }
 {
  bone1Name=dyn_pouch_bk_01_MSTR
  bone2Name=pelvis
  swing1LimitDegrees=45
  swing1Motion=Locked
  swing2LimitDegrees=45
  swing2Motion=Locked
  twistLimitDegrees=45
  twistMotion=Locked
 }
 {bone1Name=dyn_cannister_01_r, bone2Name=dyn_cannister_01_r_MSTR, swing1LimitDegrees=45, swing2LimitDegrees=45, twistLimitDegrees=5}
 {bone1Name=dyn_strap_bk_03_r, bone2Name=dyn_strap_bk_02_r, swing1LimitDegrees=45, swing2LimitDegrees=45, twistLimitDegrees=5}
 {bone1Name=dyn_strap_bk_04_r, bone2Name=dyn_strap_bk_03_r, swing1LimitDegrees=75, swing2LimitDegrees=75, twistLimitDegrees=5}
 {bone1Name=dyn_strap_bk_02_r, bone2Name=dyn_strap_bk_01_r, swing1LimitDegrees=25, swing2LimitDegrees=25, twistLimitDegrees=5}
 {bone1Name=dyn_strap_bk_01_r, bone2Name=dyn_pouch_bk_01, swing1LimitDegrees=20, swing2LimitDegrees=20, twistLimitDegrees=5}
 {
  bone1Name=calf_l
  bone2Name=thigh_l
  swing1LimitDegrees=45
  swing1Motion=Free
  swing2LimitDegrees=45
  swing2Motion=Free
  twistLimitDegrees=45
  twistMotion=Free
 }
 {
  bone1Name=thigh_l
  bone2Name=pelvis
  swing1LimitDegrees=45
  swing1Motion=Free
  swing2LimitDegrees=45
  swing2Motion=Free
  twistLimitDegrees=45
  twistMotion=Free
 }
 {
  bone1Name=thigh_r
  bone2Name=pelvis
  swing1LimitDegrees=45
  swing1Motion=Free
  swing2LimitDegrees=45
  swing2Motion=Free
  twistLimitDegrees=45
  twistMotion=Free
 }
 {
  bone1Name=calf_r
  bone2Name=thigh_r
  swing1LimitDegrees=45
  swing1Motion=Free
  swing2LimitDegrees=45
  swing2Motion=Free
  twistLimitDegrees=45
  twistMotion=Free
 }
 {bone1Name=dyn_strap_bk_01_l, bone2Name=dyn_pouch_bk_01, swing1LimitDegrees=20, swing2LimitDegrees=20, twistLimitDegrees=5}
 {bone1Name=dyn_pouch_fr_01_r, bone2Name=dyn_pouch_fr_01_r_MSTR, swing1LimitDegrees=45, swing2LimitDegrees=45, twistLimitDegrees=5}
 {
  bone1Name=dyn_pouch_01_r
  bone2Name=dyn_pouch_fr_01_r
  swing1LimitDegrees=45
  swing1Motion=Free
  swing2LimitDegrees=45
  swing2Motion=Free
  twistLimitDegrees=45
  twistMotion=Free
 }
 {
  bone1Name=dyn_pouch_01_r
  bone2Name=dyn_pouch_bk_01_r
  swing1LimitDegrees=45
  swing1Motion=Free
  swing2LimitDegrees=45
  swing2Motion=Free
  twistLimitDegrees=45
  twistMotion=Free
 }
 {
  bone1Name=thigh_r
  bone2Name=dyn_pouch_01_r
  swing1LimitDegrees=45
  swing1Motion=Free
  swing2LimitDegrees=45
  swing2Motion=Free
  twistLimitDegrees=45
  twistMotion=Free
 }
 {
  bone1Name=thigh_r
  bone2Name=dyn_pouch_fr_01_r
  swing1LimitDegrees=45
  swing1Motion=Free
  swing2LimitDegrees=45
  swing2Motion=Free
  twistLimitDegrees=45
  twistMotion=Free
 }
 {bone1Name=dyn_pouch_bk_01_r, bone2Name=pelvis, swing1LimitDegrees=45, swing2LimitDegrees=45, twistLimitDegrees=5}
 {bone1Name=dyn_pouch_01_r, bone2Name=dyn_pouch_01_r_MSTR, swing1LimitDegrees=75, swing2LimitDegrees=75, twistLimitDegrees=5}
 {bone1Name=dyn_pouch_04_l, bone2Name=dyn_pouch_03_l, swing1LimitDegrees=60, swing2LimitDegrees=60, twistLimitDegrees=10}
 {
  bone1Name=dyn_pouch_02_r
  bone2Name=dyn_pouch_01_r_MSTR
  swing1LimitDegrees=45
  swing1Motion=Free
  swing2LimitDegrees=45
  swing2Motion=Free
  twistLimitDegrees=45
  twistMotion=Free
 }
 {bone1Name=dyn_pouch_03_r, bone2Name=dyn_pouch_02_r, swing1LimitDegrees=20, swing2LimitDegrees=20, twistLimitDegrees=15}
 {bone1Name=dyn_pouch_04_r, bone2Name=dyn_pouch_03_r, swing1LimitDegrees=45, swing2LimitDegrees=45, twistLimitDegrees=10}
 {
  bone1Name=dyn_pouch_02_r
  bone2Name=dyn_pouch_01_r
  swing1LimitDegrees=45
  swing1Motion=Free
  swing2LimitDegrees=45
  swing2Motion=Free
  twistLimitDegrees=45
  twistMotion=Free
 }
 {bone1Name=dyn_pouch_02_l, bone2Name=dyn_pouch_01_l, swing1LimitDegrees=2, swing2LimitDegrees=10, twistLimitDegrees=2}
 {
  bone1Name=dyn_pouch_01_l
  bone2Name=thigh_l
  swing1LimitDegrees=45
  swing1Motion=Free
  swing2LimitDegrees=45
  swing2Motion=Free
  twistLimitDegrees=45
  twistMotion=Free
 }
 {bone1Name=dyn_pouch_03_l, bone2Name=dyn_pouch_02_l, swing1LimitDegrees=20, swing2LimitDegrees=20, twistLimitDegrees=15}
 {
  bone1Name=thigh_l
  bone2Name=dyn_pouch_fr_01_l
  swing1LimitDegrees=45
  swing1Motion=Free
  swing2LimitDegrees=45
  swing2Motion=Free
  twistLimitDegrees=45
  twistMotion=Free
 }
 {bone1Name=dyn_pouch_fr_01_l, bone2Name=dyn_pouch_fr_01_l_MSTR, swing1LimitDegrees=75, swing2LimitDegrees=75, twistLimitDegrees=5}
 {
  bone1Name=dyn_pouch_02_l
  bone2Name=dyn_pouch_fr_01_l
  swing1LimitDegrees=45
  swing1Motion=Free
  swing2LimitDegrees=45
  swing2Motion=Free
  twistLimitDegrees=45
  twistMotion=Free
 }
]
