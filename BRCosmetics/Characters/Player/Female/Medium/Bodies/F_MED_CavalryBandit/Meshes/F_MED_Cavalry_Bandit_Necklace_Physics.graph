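# Physics asset for the necklace on the female medium Cavalry Bandit player cosmetic, defining its bodies and joint limits for physical motion.
# A five-bone chain of sphere, capsules and an end box links to spine_05 through dyn_necklace_02, with swing limits widening toward the tip and twist mostly locked.
# Head sphere and overlapping spine capsules approximate the neck and chest as collision shapes, joined by free-rotation constraints.
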PhysicsAsset /BRCosmetics/Characters/Player/Female/Medium/Bodies/F_MED_CavalryBandit/Meshes/F_MED_Cavalry_Bandit_Necklace_Physics
bodies {
 dyn_necklace_01_MSTR [{shapeType=Sphere, radius=1, shapeName=dyn_necklace_01_MSTR_sphere}]
 dyn_necklace_02 [{shapeType=Capsule, length=4, radius=1, shapeName=dyn_necklace_02_capsule, center=<1 0 0>, rotation=<rot 0 0 90>}]
 dyn_necklace_03 [{shapeType=Capsule, length=2, radius=1, shapeName=dyn_necklace_03_capsule, rotation=<rot 0 0 -90>}]
 dyn_necklace_04 [{shapeType=Capsule, length=3, radius=1.2, shapeName=dyn_necklace_04_capsule, rotation=<rot 0 0 90>}]
 dyn_necklace_05 [{
   shapeType=Box
   extentX=4.5696
   extentY=1.423
   extentZ=4.0613
   shapeName=dyn_necklace_05_box
   center=<1.7795 0.0021 -0.4112>
   rotation=<rot -9.9973 -4.1332 -88.7143>
  }]
 head [{shapeType=Sphere, radius=8.755, shapeName=head_sphere, center=<0.2578 2.2273 0.0013>}]
 spine_04 [
  {shapeType=Capsule, length=6.25, radius=11.4, center=<0.1 0 0>}
  {shapeType=Capsule, length=4, radius=11.4, shapeName=spine_04_capsule, center=<-2.5281 2.7275 0>}
 ]
 spine_05 [
  {shapeType=Capsule, length=13.7034, radius=7.1}
  {shapeType=Capsule, length=15, radius=5.5, shapeName=undernecklace_orig, center=<-6.6693 7.8194 -0.4>}
  {shapeType=Capsule, length=11.7034, radius=9.1, shapeName=2, center=<-4.5232 1.2631 1.389>, rotation=<rot 90 172.875 -144.9369>}
  {shapeType=Capsule, length=11.7034, radius=9.1, shapeName=_2_double, center=<-4.5232 1.2631 -1.573>, rotation=<rot 90 172.875 -144.9369>}
 ]
}
constraints [
 {bone1Name=dyn_necklace_02, bone2Name=spine_05, swing1LimitDegrees=25, swing2LimitDegrees=25, twistLimitDegrees=40, twistMotion=Locked}
 {bone1Name=dyn_necklace_03, bone2Name=dyn_necklace_02, swing1LimitDegrees=50, swing2LimitDegrees=50, twistLimitDegrees=10, twistMotion=Locked}
 {bone1Name=dyn_necklace_04, bone2Name=dyn_necklace_03, swing1LimitDegrees=60, swing2LimitDegrees=60, twistLimitDegrees=10}
 {bone1Name=dyn_necklace_05, bone2Name=dyn_necklace_04, swing1LimitDegrees=75, swing2LimitDegrees=75, twistLimitDegrees=10, twistMotion=Locked}
 {
  bone1Name=spine_05
  bone2Name=spine_04
  swing1LimitDegrees=45
  swing1Motion=Free
  swing2LimitDegrees=45
  swing2Motion=Free
  twistLimitDegrees=45
  twistMotion=Free
 }
 {
  bone1Name=dyn_necklace_01_MSTR
  bone2Name=dyn_necklace_02
  swing1LimitDegrees=20
  swing2LimitDegrees=30
  twistLimitDegrees=40
  twistMotion=Locked
 }
 {
  bone1Name=head
  bone2Name=spine_05
  swing1LimitDegrees=45
  swing1Motion=Free
  swing2LimitDegrees=45
  swing2Motion=Free
  twistLimitDegrees=45
  twistMotion=Free
 }
]
