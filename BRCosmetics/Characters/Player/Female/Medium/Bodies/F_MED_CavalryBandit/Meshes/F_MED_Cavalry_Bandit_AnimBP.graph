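# Female medium Cavalry Bandit body AnimBP that copies its attached parent's pose and adds secondary motion to pouches, straps, necklace and gloves.
# Bone constraints prepare the pose, and cached 'PreSimulate' feeds RigidBody nodes with moving, crouch or front-end physics assets.
# Variable-driven corrective offsets form 'SimBase', where bRigidBodyEnabled selects between that result and added AnimDynamics straps and pouches.
# Final crouch-walk and directional pouch corrections precede output.
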AnimBlueprint /BRCosmetics/Characters/Player/Female/Medium/Bodies/F_MED_CavalryBandit/Meshes/F_MED_Cavalry_Bandit_AnimBP
target_skeleton=&/F_MED_Cavalry_Bandit_Skeleton
graphs [
 AnimGraph {nodes [
   n1 = Look At - Bone: dyn_pouch_fr_01_l(Alpha=0) {
    alphaBoolBlend {blendOption=Linear}
    boneToModify {boneName=dyn_pouch_fr_01_l}
    lookAtLocation=<0 8>
    lookAtTarget {boneReference {boneName=dyn_pouch_fr_01_l_aim}}
    lookAt_Axis {axis=<1 0>}
    lookUp_Axis {axis {x=0, z=1}}
   }
   n2 = Look At - Bone: dyn_pouch_fr_01_r(ComponentPose=@n1, Alpha=0) {
    alphaBoolBlend {blendOption=Linear}
    boneToModify {boneName=dyn_pouch_fr_01_r}
    lookAtLocation=<0 1>
    lookAtTarget {boneReference {boneName=dyn_pouch_fr_01_r_aim}}
    lookAt_Axis {axis=<-1 0>}
    lookUp_Axis {axis {x=0, z=1}}
   }
   n3 = "Transform (Modify) Bone - Bone: dyn_pouch_fr_01_l_MSTR"(ComponentPose=@n2, Alpha=1, Translation="0.000,1,0.000", Rotation="0,0,0", Scale="1,1,1") {
    rotationMode=BMM_Additive
    rotationSpace=BCS_BoneSpace
    translationMode=BMM_Additive
    translationSpace=BCS_BoneSpace
    alphaBoolBlend {blendOption=Linear}
    boneToModify {boneName=dyn_pouch_fr_01_l_MSTR}
    translation {y=1}
   }
   "Transform (Modify) Bone - Bone: dyn_pouch_fr_01_r_MSTR"(ComponentPose=@n3, Alpha=0, Translation="0.000,-1,0.000", Rotation="0.000,5.000,0.000", Scale="1,1,1") {
    rotationMode=BMM_Additive
    rotationSpace=BCS_BoneSpace
    translationMode=BMM_Additive
    translationSpace=BCS_BoneSpace
    alphaBoolBlend {blendOption=Linear}
    boneToModify {boneName=dyn_pouch_fr_01_r_MSTR}
    rotation {yaw=5.0002}
    translation {y=-1}
   }
   n4 = "Transform (Modify) Bone - Bone: dyn_pouch_01_r_MSTR"!(ComponentPose, bAlphaBoolEnabled:bool=True, Translation="0,0,0", Rotation="10,0.000,0.000", Scale="1,1,1") -> (Pose) {
    alphaInputType=Bool
    rotationMode=BMM_Additive
    rotationSpace=BCS_BoneSpace
    scaleMode=BMM_Additive
    scaleSpace=BCS_BoneSpace
    translationMode=BMM_Additive
    translationSpace=BCS_BoneSpace
    alphaBoolBlend {blendOption=Linear}
    boneToModify {boneName=dyn_pouch_01_r_MSTR}
    rotation {pitch=10}
   }
   "Transform (Modify) Bone - Bone: dyn_pouch_03_r"!(ComponentPose=@n4, bAlphaBoolEnabled:bool=True, Translation="0,0,0", Rotation="-20,0,0", Scale="1,1,1") -> (Pose) {
    alphaInputType=Bool
    rotationMode=BMM_Additive
    rotationSpace=BCS_BoneSpace
    scaleMode=BMM_Additive
    scaleSpace=BCS_BoneSpace
    translationMode=BMM_Additive
    translationSpace=BCS_BoneSpace
    alphaBoolBlend {blendOption=Linear}
    boneToModify {boneName=dyn_pouch_03_r}
    rotation {pitch=-20}
   }
   n5 = "Transform (Modify) Bone - Bone: dyn_strap_bk_01_l"(ComponentPose=Local To Component(LocalPose=Use cached pose 'SimBase'()), Alpha=1, Translation="0,0,0", Rotation="20,0,0", Scale="1,1,1") {
    alpha=0
    rotationMode=BMM_Additive
    rotationSpace=BCS_BoneSpace
    scaleMode=BMM_Additive
    scaleSpace=BCS_BoneSpace
    translationMode=BMM_Additive
    translationSpace=BCS_BoneSpace
    alphaBoolBlend {blendOption=Linear}
    boneToModify {boneName=dyn_strap_bk_01_l}
    rotation {pitch=20}
   }
   n6 = "Transform (Modify) Bone - Bone: dyn_strap_bk_01_r"(ComponentPose=@n5, Alpha=1, Translation="0,0,0", Rotation="20,0,0", Scale="1,1,1") {
    alpha=0
    rotationMode=BMM_Additive
    rotationSpace=BCS_BoneSpace
    scaleMode=BMM_Additive
    scaleSpace=BCS_BoneSpace
    translationMode=BMM_Additive
    translationSpace=BCS_BoneSpace
    alphaBoolBlend {blendOption=Linear}
    boneToModify {boneName=dyn_strap_bk_01_r}
    rotation {pitch=20}
   }
   n7 = "Transform (Modify) Bone - Bone: dyn_pouch_03_l"(ComponentPose=@n6, Alpha=1, Translation="0,0,0", Rotation="60.000,0,0", Scale="1,1,1") {
    alpha=0
    rotationMode=BMM_Additive
    rotationSpace=BCS_BoneSpace
    scaleMode=BMM_Additive
    scaleSpace=BCS_BoneSpace
    translationMode=BMM_Additive
    translationSpace=BCS_BoneSpace
    alphaBoolBlend {blendOption=Linear}
    boneToModify {boneName=dyn_pouch_03_l}
    rotation {pitch=60}
   }
   n8 = "Transform (Modify) Bone - Bone: dyn_pouch_03_r"!(ComponentPose=@n7, Alpha:float=1, Translation="0,0,0", Rotation="60,0,0", Scale="1,1,1") -> (Pose) {
    alpha=0
    rotationMode=BMM_Additive
    rotationSpace=BCS_BoneSpace
    scaleMode=BMM_Additive
    scaleSpace=BCS_BoneSpace
    translationMode=BMM_Additive
    translationSpace=BCS_BoneSpace
    alphaBoolBlend {blendOption=Linear}
    boneToModify {boneName=dyn_pouch_03_r}
    rotation {pitch=60}
   }
   n9 = "Constraint - dyn_pouch_bk_01 (P)"(ComponentPose=@n8, Alpha=1, ConstraintWeights_0=1) {
    alpha=0
    alphaBoolBlend {blendOption=Linear}
    boneToModify {boneName=dyn_pouch_bk_01}
    constraintSetup [{targetBone {boneName=pelvis}}]
    constraintWeights [1]
   }
   n10 = Get FPS_Alpha() -> (FPS_Alpha:double)
   n11 = "Transform (Modify) Bone - Bone: dyn_necklace_04"(ComponentPose=@n9, Alpha=@n10, Translation="0,0,0.2", Rotation="7,0,0", Scale="1,1,1") {
    alpha=0
    rotationMode=BMM_Additive
    rotationSpace=BCS_BoneSpace
    scaleMode=BMM_Additive
    scaleSpace=BCS_BoneSpace
    translationMode=BMM_Additive
    translationSpace=BCS_BoneSpace
    alphaBoolBlend {blendOption=Linear}
    alphaScaleBiasClamp {bInterpResult=true, interpSpeedDecreasing=3, interpSpeedIncreasing=3}
    boneToModify {boneName=dyn_necklace_04}
    rotation {pitch=7}
    translation {z=0.2}
   }
   n12 = "Transform (Modify) Bone - Bone: dyn_necklace_05"(ComponentPose=@n11, Alpha=@n10, Translation="0,0,0", Rotation="-8,0,0", Scale="1,1,1") {
    alpha=0
    rotationMode=BMM_Additive
    rotationSpace=BCS_BoneSpace
    scaleMode=BMM_Additive
    scaleSpace=BCS_BoneSpace
    translationMode=BMM_Additive
    translationSpace=BCS_BoneSpace
    alphaBoolBlend {blendOption=Linear}
    alphaScaleBiasClamp {bInterpResult=true, interpSpeedDecreasing=3, interpSpeedIncreasing=3}
    boneToModify {boneName=dyn_necklace_05}
    rotation {pitch=-8}
   }
   n13 = Component To Local(ComponentPose=@n12)
   n14 = "Constraint - dyn_pouch_03_l (P)"(ComponentPose=Local To Component(LocalPose=Use cached pose 'SimBase'()), Alpha=1, ConstraintWeights_0=1) {
    alpha=0
    alphaBoolBlend {blendOption=Linear}
    boneToModify {boneName=dyn_pouch_03_l}
    constraintSetup [{targetBone {boneName=thigh_l}}]
    constraintWeights [1]
   }
   n15 = "Constraint - dyn_pouch_03_r (P)"(ComponentPose=@n14, Alpha=1, ConstraintWeights_0=1) {
    alpha=0
    alphaBoolBlend {blendOption=Linear}
    boneToModify {boneName=dyn_pouch_03_r}
    constraintSetup [{targetBone {boneName=thigh_r}}]
    constraintWeights [1]
   }
   n16 = "Constraint - dyn_strap_bk_01_l (P)"(ComponentPose=@n15, Alpha=1, ConstraintWeights_0=1) {
    alpha=0
    alphaBoolBlend {blendOption=Linear}
    boneToModify {boneName=dyn_strap_bk_01_l}
    constraintSetup [{targetBone {boneName=thigh_l}}]
    constraintWeights [1]
   }
   n17 = "Constraint - dyn_strap_bk_01_r (P)"(ComponentPose=@n16, Alpha=1, ConstraintWeights_0=1) {
    alpha=0
    alphaBoolBlend {blendOption=Linear}
    boneToModify {boneName=dyn_strap_bk_01_r}
    constraintSetup [{targetBone {boneName=thigh_r}}]
    constraintWeights [1]
   }
   n18 = "Transform (Modify) Bone - Bone: dyn_strap_bk_01_l"(ComponentPose=@n17, Alpha=1, Translation="0,0,0", Rotation="20,0,0", Scale="1,1,1") {
    alpha=0
    rotationMode=BMM_Additive
    rotationSpace=BCS_BoneSpace
    scaleMode=BMM_Additive
    scaleSpace=BCS_BoneSpace
    translationMode=BMM_Additive
    translationSpace=BCS_BoneSpace
    alphaBoolBlend {blendOption=Linear}
    boneToModify {boneName=dyn_strap_bk_01_l}
    rotation {pitch=20}
   }
   n19 = "Transform (Modify) Bone - Bone: dyn_strap_bk_01_r"(ComponentPose=@n18, Alpha=1, Translation="0,0,0", Rotation="20.000,0.000,0.000", Scale="1,1,1") {
    alpha=0
    rotationMode=BMM_Additive
    rotationSpace=BCS_BoneSpace
    scaleMode=BMM_Additive
    scaleSpace=BCS_BoneSpace
    translationMode=BMM_Additive
    translationSpace=BCS_BoneSpace
    alphaBoolBlend {blendOption=Linear}
    boneToModify {boneName=dyn_strap_bk_01_r}
    rotation {pitch=20}
   }
   n20 = Anim Dynamics - Bone: dyn_strap_bk_02_l(ComponentPose=@n19, Alpha=1, ExternalForce="0,0,0") {
    angularDampingOverride=0.9999
    angularSpringConstant=100
    bAngularSpring=true
    bOverrideAngularDamping=true
    simulationSpace=RootRelative
    alphaBoolBlend {blendOption=Linear}
    boundBone {boneName=dyn_strap_bk_02_l}
    componentAppliedLinearAccClamp=<3500 3500 3500>
    componentLinearAccScale=<1.2 1.2 1.2>
    componentLinearVelScale=<1.2 1.2 1.2>
    physicsBodyDefinitions [{
      sphereCollisionRadius=0
      boundBone {boneName=dyn_strap_bk_02_l}
      boxExtents=<20 10 10>
      constraintSetup {angularLimitsMax=<1 10 15>, angularLimitsMin=<-1 -30 -15>, angularTarget=<0 0 0>}
      localJointOffset=<2 0 0>
     }]
    relativeSpaceBone {boneName=pelvis}
   }
   n21 = Anim Dynamics - Bone: dyn_strap_bk_03_l(ComponentPose=@n20, Alpha=1, ExternalForce="0,0,0") {
    angularDampingOverride=0.9999
    angularSpringConstant=100
    bAngularSpring=true
    bOverrideAngularDamping=true
    simulationSpace=RootRelative
    alphaBoolBlend {blendOption=Linear}
    boundBone {boneName=dyn_strap_bk_03_l}
    componentAppliedLinearAccClamp=<3500 3500 3500>
    componentLinearAccScale=<1.2 1.2 1.2>
    componentLinearVelScale=<1.2 1.2 1.2>
    physicsBodyDefinitions [{
      sphereCollisionRadius=0
      boundBone {boneName=dyn_strap_bk_03_l}
      boxExtents=<20 10 10>
      constraintSetup {angularLimitsMax=<1 10 15>, angularLimitsMin=<-1 -30 -15>, angularTarget=<0 0 0>}
      localJointOffset=<2 0 0>
     }]
    relativeSpaceBone {boneName=pelvis}
   }
   n22 = Anim Dynamics - Bone: dyn_strap_bk_04_l(ComponentPose=@n21, Alpha=1, ExternalForce="0,0,0") {
    angularDampingOverride=0.9999
    angularSpringConstant=100
    bAngularSpring=true
    bOverrideAngularDamping=true
    simulationSpace=RootRelative
    alphaBoolBlend {blendOption=Linear}
    boundBone {boneName=dyn_strap_bk_04_l}
    componentAppliedLinearAccClamp=<3500 3500 3500>
    componentLinearAccScale=<1.2 1.2 1.2>
    componentLinearVelScale=<1.2 1.2 1.2>
    physicsBodyDefinitions [{
      sphereCollisionRadius=0
      boundBone {boneName=dyn_strap_bk_04_l}
      boxExtents=<20 10 10>
      constraintSetup {angularLimitsMax=<1 5 15>, angularLimitsMin=<-1 -30 -15>, angularTarget=<0 0 0>}
      localJointOffset=<2 0 0>
     }]
    relativeSpaceBone {boneName=pelvis}
   }
   n23 = Anim Dynamics - Bone: dyn_strap_bk_02_r(ComponentPose=@n22, Alpha=1, ExternalForce="0,0,0") {
    angularDampingOverride=0.9999
    angularSpringConstant=100
    bAngularSpring=true
    bOverrideAngularDamping=true
    simulationSpace=RootRelative
    alphaBoolBlend {blendOption=Linear}
    boundBone {boneName=dyn_strap_bk_02_r}
    componentAppliedLinearAccClamp=<3500 3500 3500>
    componentLinearAccScale=<1.2 1.2 1.2>
    componentLinearVelScale=<1.2 1.2 1.2>
    physicsBodyDefinitions [{
      sphereCollisionRadius=0
      boundBone {boneName=dyn_strap_bk_02_r}
      boxExtents=<20 10 10>
      constraintSetup {angularLimitsMax=<1 5 15>, angularLimitsMin=<-1 -30 -15>, angularTarget=<0 0 0>}
      localJointOffset=<2 0 0>
     }]
    relativeSpaceBone {boneName=pelvis}
   }
   n24 = Anim Dynamics - Bone: dyn_strap_bk_03_r(ComponentPose=@n23, Alpha=1, ExternalForce="0,0,0") {
    angularDampingOverride=0.9999
    angularSpringConstant=100
    bAngularSpring=true
    bOverrideAngularDamping=true
    simulationSpace=RootRelative
    alphaBoolBlend {blendOption=Linear}
    boundBone {boneName=dyn_strap_bk_03_r}
    componentAppliedLinearAccClamp=<3500 3500 3500>
    componentLinearAccScale=<1.2 1.2 1.2>
    componentLinearVelScale=<1.2 1.2 1.2>
    physicsBodyDefinitions [{
      sphereCollisionRadius=0
      boundBone {boneName=dyn_strap_bk_03_r}
      boxExtents=<20 10 10>
      constraintSetup {angularLimitsMax=<1 5 15>, angularLimitsMin=<-1 -30 -15>, angularTarget=<0 0 0>}
      localJointOffset=<2 0 0>
     }]
    relativeSpaceBone {boneName=pelvis}
   }
   n25 = Anim Dynamics - Bone: dyn_strap_bk_04_r(ComponentPose=@n24, Alpha=1, ExternalForce="0,0,0") {
    angularDampingOverride=0.9999
    angularSpringConstant=100
    bAngularSpring=true
    bOverrideAngularDamping=true
    simulationSpace=RootRelative
    alphaBoolBlend {blendOption=Linear}
    boundBone {boneName=dyn_strap_bk_04_r}
    componentAppliedLinearAccClamp=<3500 3500 3500>
    componentLinearAccScale=<1.2 1.2 1.2>
    componentLinearVelScale=<1.2 1.2 1.2>
    physicsBodyDefinitions [{
      sphereCollisionRadius=0
      boundBone {boneName=dyn_strap_bk_04_r}
      boxExtents=<20 10 10>
      constraintSetup {angularLimitsMax=<1 5 15>, angularLimitsMin=<-1 -30 -15>, angularTarget=<0 0 0>}
      localJointOffset=<2 0 0>
     }]
    relativeSpaceBone {boneName=pelvis}
   }
   n26 = Anim Dynamics - Bone: dyn_pouch_03_l(ComponentPose=@n25, Alpha=1, ExternalForce="0,0,0") {
    angularDampingOverride=0.9999
    angularSpringConstant=100
    bAngularSpring=true
    bOverrideAngularDamping=true
    simulationSpace=RootRelative
    alphaBoolBlend {blendOption=Linear}
    boundBone {boneName=dyn_pouch_03_l}
    componentAppliedLinearAccClamp=<3500 3500 3500>
    componentLinearAccScale=<1.2 1.2 1.2>
    componentLinearVelScale=<1.2 1.2 1.2>
    physicsBodyDefinitions [{
      sphereCollisionRadius=0
      boundBone {boneName=dyn_pouch_03_l}
      boxExtents=<20 10 10>
      constraintSetup {angularLimitsMax=<1 15 20>, angularLimitsMin=<-1 -15 -30>, angularTarget=<0 0 0>}
      localJointOffset=<2 0 0>
     }]
    relativeSpaceBone {boneName=pelvis}
   }
   n27 = Anim Dynamics - Bone: dyn_pouch_04_l(ComponentPose=@n26, Alpha=1, ExternalForce="0,0,0") {
    angularDampingOverride=0.9999
    angularSpringConstant=100
    bAngularSpring=true
    bOverrideAngularDamping=true
    simulationSpace=RootRelative
    alphaBoolBlend {blendOption=Linear}
    boundBone {boneName=dyn_pouch_04_l}
    componentAppliedLinearAccClamp=<3500 3500 3500>
    componentLinearAccScale=<1.2 1.2 1.2>
    componentLinearVelScale=<1.2 1.2 1.2>
    physicsBodyDefinitions [{
      sphereCollisionRadius=0
      boundBone {boneName=dyn_pouch_04_l}
      boxExtents=<20 10 10>
      constraintSetup {angularLimitsMax=<1 15 20>, angularLimitsMin=<-1 -15 -30>, angularTarget=<0 0 0>}
      localJointOffset=<2 0 0>
     }]
    relativeSpaceBone {boneName=pelvis}
   }
   n28 = Anim Dynamics - Bone: dyn_pouch_05_l(ComponentPose=@n27, Alpha=1, ExternalForce="0,0,0") {
    angularDampingOverride=0.9999
    angularSpringConstant=100
    bAngularSpring=true
    bOverrideAngularDamping=true
    simulationSpace=RootRelative
    alphaBoolBlend {blendOption=Linear}
    boundBone {boneName=dyn_pouch_05_l}
    componentAppliedLinearAccClamp=<3500 3500 3500>
    componentLinearAccScale=<1.2 1.2 1.2>
    componentLinearVelScale=<1.2 1.2 1.2>
    physicsBodyDefinitions [{
      sphereCollisionRadius=0
      boundBone {boneName=dyn_pouch_05_l}
      boxExtents=<20 10 10>
      constraintSetup {angularLimitsMax=<1 15 20>, angularLimitsMin=<-1 -15 -30>, angularTarget=<0 0 0>}
      localJointOffset=<2 0 0>
     }]
    relativeSpaceBone {boneName=pelvis}
   }
   n29 = Anim Dynamics - Bone: dyn_pouch_03_r(ComponentPose=@n28, Alpha=1, ExternalForce="0,0,0") {
    angularDampingOverride=0.9999
    angularSpringConstant=100
    bAngularSpring=true
    bOverrideAngularDamping=true
    simulationSpace=RootRelative
    alphaBoolBlend {blendOption=Linear}
    boundBone {boneName=dyn_pouch_03_r}
    componentAppliedLinearAccClamp=<3500 3500 3500>
    componentLinearAccScale=<1.2 1.2 1.2>
    componentLinearVelScale=<1.2 1.2 1.2>
    physicsBodyDefinitions [{
      sphereCollisionRadius=0
      boundBone {boneName=dyn_pouch_03_r}
      boxExtents=<20 10 10>
      constraintSetup {angularLimitsMax=<1 15 30>, angularLimitsMin=<-1 -15 -20>, angularTarget=<0 0 0>}
      localJointOffset=<-2 0 0>
     }]
    relativeSpaceBone {boneName=pelvis}
   }
   n30 = Anim Dynamics - Bone: dyn_pouch_04_r(ComponentPose=@n29, Alpha=1, ExternalForce="0,0,0") {
    angularDampingOverride=0.9999
    angularSpringConstant=100
    bAngularSpring=true
    bOverrideAngularDamping=true
    simulationSpace=RootRelative
    alphaBoolBlend {blendOption=Linear}
    boundBone {boneName=dyn_pouch_04_r}
    componentAppliedLinearAccClamp=<3500 3500 3500>
    componentLinearAccScale=<1.2 1.2 1.2>
    componentLinearVelScale=<1.2 1.2 1.2>
    physicsBodyDefinitions [{
      sphereCollisionRadius=0
      boundBone {boneName=dyn_pouch_04_r}
      boxExtents=<20 10 10>
      constraintSetup {angularLimitsMax=<1 15 30>, angularLimitsMin=<-1 -15 -20>, angularTarget=<0 0 0>}
      localJointOffset=<-2 0 0>
     }]
    relativeSpaceBone {boneName=pelvis}
   }
   n31 = Anim Dynamics - Bone: dyn_pouch_05_r(ComponentPose=@n30, Alpha=1, ExternalForce="0,0,0") {
    angularDampingOverride=0.9999
    angularSpringConstant=100
    bAngularSpring=true
    bOverrideAngularDamping=true
    simulationSpace=RootRelative
    alphaBoolBlend {blendOption=Linear}
    boundBone {boneName=dyn_pouch_05_r}
    componentAppliedLinearAccClamp=<3500 3500 3500>
    componentLinearAccScale=<1.2 1.2 1.2>
    componentLinearVelScale=<1.2 1.2 1.2>
    physicsBodyDefinitions [{
      sphereCollisionRadius=0
      boundBone {boneName=dyn_pouch_05_r}
      boxExtents=<20 10 10>
      constraintSetup {angularLimitsMax=<1 15 30>, angularLimitsMin=<-1 -15 -20>, angularTarget=<0 0 0>}
      localJointOffset=<-2 0 0>
     }]
    relativeSpaceBone {boneName=pelvis}
   }
   n32 = Component To Local(ComponentPose=@n31)
   n33 = Get bRigidBodyEnabled()
   n34 = Get Crouch_walk() -> (Crouch_walk:double)
   n35 = Look At - Bone: dyn_pouch_01_l_MSTR(
    ComponentPose=Local To Component(LocalPose=Blend Poses by bool(BlendPose_0=@n13, BlendPose_1=@n32, bActiveValue=@n33) {blendType=Linear})
    Alpha=@n34
   ) {
    alphaBoolBlend {blendOption=Linear}
    alphaScaleBiasClamp {bInterpResult=true, interpSpeedDecreasing=7, interpSpeedIncreasing=7}
    boneToModify {boneName=dyn_pouch_01_l_MSTR}
    lookAtLocation=<1.6604 -2.8527 -5.4647>
    lookAtTarget {boneReference {boneName=thigh_l}}
    lookAt_Axis {axis=<1 0>}
   }
   n36 = "Transform (Modify) Bone - Bone: dyn_pouch_01_l_MSTR"(ComponentPose=@n35, Alpha=@n34, Translation="0,0,0", Rotation="0.028,9.995,-0.011", Scale="1,1,1") {
    rotationMode=BMM_Additive
    rotationSpace=BCS_BoneSpace
    translationMode=BMM_Additive
    translationSpace=BCS_BoneSpace
    alphaBoolBlend {blendOption=Linear}
    alphaScaleBiasClamp {bInterpResult=true, interpSpeedDecreasing=5, interpSpeedIncreasing=7}
    boneToModify {boneName=dyn_pouch_01_l_MSTR}
    rotation=<rot 0.028 9.995 -0.011>
   }
   n37 = Get Left_Direct() -> (Left_Direct:double)
   n38 = Look At - Bone: dyn_pouch_01_r_MSTR(ComponentPose=@n36, Alpha=@n37) {
    alphaBoolBlend {blendOption=Linear}
    alphaScaleBiasClamp {bInterpResult=true, interpSpeedDecreasing=7, interpSpeedIncreasing=7}
    boneToModify {boneName=dyn_pouch_01_r_MSTR}
    lookAtLocation=<-0.8076 -4.8803 -3.5599>
    lookAtTarget {boneReference {boneName=pelvis}}
    lookAt_Axis {axis=<1 0>}
   }
   n39 = "Transform (Modify) Bone - Bone: dyn_pouch_01_r_MSTR"!(ComponentPose=@n38, Alpha:float=@n37, Translation="-2.396,-1.488,5.460", Rotation="11.567,10.068,-8.131", Scale="1,1,1") -> (Pose) {
    rotationMode=BMM_Additive
    rotationSpace=BCS_BoneSpace
    translationMode=BMM_Additive
    translationSpace=BCS_BoneSpace
    alphaBoolBlend {blendOption=Linear}
    alphaScaleBiasClamp {bInterpResult=true, interpSpeedDecreasing=5, interpSpeedIncreasing=7}
    boneToModify {boneName=dyn_pouch_01_r_MSTR}
    rotation=<rot 11.567 10.068 -8.131>
    translation=<-2.396 -1.488 5.46>
   }
   n40 = Get Right_Direct() -> (Right_Direct:double)
   n41 = Look At - Bone: dyn_pouch_01_r_MSTR(ComponentPose=@n39, Alpha=@n40) {
    alphaBoolBlend {blendOption=Linear}
    alphaScaleBiasClamp {bInterpResult=true, interpSpeedDecreasing=7, interpSpeedIncreasing=7}
    boneToModify {boneName=dyn_pouch_01_r_MSTR}
    lookAtLocation=<10.3263 10.3387 -28.3252>
    lookAtTarget {boneReference {boneName=pelvis}}
    lookAt_Axis {axis=<1 0>}
   }
   n42 = "Transform (Modify) Bone - Bone: dyn_pouch_01_r_MSTR"(ComponentPose=@n41, Alpha=@n40, Translation="1.955,-2.483,-2.152", Rotation="3.375,-9.404,9.722", Scale="1,1,1") {
    rotationMode=BMM_Additive
    rotationSpace=BCS_BoneSpace
    translationMode=BMM_Additive
    translationSpace=BCS_BoneSpace
    alphaBoolBlend {blendOption=Linear}
    alphaScaleBiasClamp {bInterpResult=true, interpSpeedDecreasing=5, interpSpeedIncreasing=7}
    boneToModify {boneName=dyn_pouch_01_r_MSTR}
    rotation=<rot 3.375 -9.404 9.722>
    translation=<1.955 -2.483 -2.152>
   }
   Output Pose(Result=Component To Local(ComponentPose=@n42)) {layerGroup=None}
   n43 = Get Belt_l_correction() -> (Belt_l_correction:double)
   n44 = Get Component Linear Acc Scale() -> (Component Linear Acc Scale:Vector)
   n45 = Get Component Linear Vel Scale() -> (Component Linear Vel Scale:Vector)
   n46 = Get FinalGravity() -> (FinalGravity:Vector)
   n47 = RigidBody!(
    ComponentPose=Local To Component(LocalPose=Use cached pose 'PreSimulate'())
    bAlphaBoolEnabled:bool=NOT Boolean(A=Get bIsCrouching())
    OverrideWorldGravity:Vector=@n46
    ExternalForce
    ComponentLinearAccScale:Vector=@n44
    ComponentLinearVelScale:Vector=@n45
   ) -> (Pose) {
    alphaInputType=Bool
    bClampLinearTranslationLimitToRefPose=true
    bOverrideWorldGravity=true
    overridePhysicsAsset=&/F_MED_Cavalry_Bandit_Body_Physics_Moving
    simulationSpace=BaseBoneSpace
    alphaBoolBlend {blendInTime=0.1, blendOption=Linear, blendOutTime=0.1}
    baseBoneRef {boneName=sim}
    componentAppliedLinearAccClamp=<3000 3000 3000>
    componentLinearAccScale=<0.5 0.5 0.5>
    componentLinearVelScale=<0.7 0.7 0.7>
    overrideWorldGravity {z=-980}
   }
   n48 = RigidBody(
    ComponentPose=@n47
    bAlphaBoolEnabled=Get bIsCrouching()
    OverrideWorldGravity=@n46
    ComponentLinearAccScale=@n44
    ComponentLinearVelScale=@n45
   ) {
    alphaInputType=Bool
    bClampLinearTranslationLimitToRefPose=true
    bOverrideWorldGravity=true
    overridePhysicsAsset=&/F_MED_Cavalry_Bandit_Body_Physics_crouch
    simulationSpace=BaseBoneSpace
    alphaBoolBlend {blendOption=Linear}
    baseBoneRef {boneName=sim}
    componentAppliedLinearAccClamp=<3000 3000 3000>
    componentLinearAccScale=<0.5 0.5 0.5>
    componentLinearVelScale=<0.7 0.7 0.7>
    overrideWorldGravity {z=-980}
   }
   n49 = Component To Local(ComponentPose=@n48)
   n50 = RigidBody!(ComponentPose=Local To Component(LocalPose=Use cached pose 'PreSimulate'()), bAlphaBoolEnabled, ExternalForce) -> (Pose) {
    alphaInputType=Bool
    bOverrideWorldGravity=true
    overridePhysicsAsset=&/F_MED_Cavalry_Bandit_Body_Physics_Moving_frontend
    simulationSpace=BaseBoneSpace
    alphaBoolBlend {blendOption=Linear}
    baseBoneRef {boneName=pelvis}
    componentAppliedLinearAccClamp=<3000 3000 3000>
    componentLinearAccScale=<0.4 0.3 0.3>
    componentLinearVelScale=<0.4 0.3 0.3>
    overrideWorldGravity {z=-980}
   }
   n51 = "Constraint - dyn_pouch_02_r (P)"(ComponentPose=@n50, Alpha=Get Pouch_02_r_correction() -> (Pouch_02_r_correction:double), ConstraintWeights_0=1) {
    alphaBoolBlend {blendOption=Linear}
    alphaScaleBiasClamp {bInterpResult=true, interpSpeedDecreasing=2}
    boneToModify {boneName=dyn_pouch_02_r}
    constraintSetup [{targetBone {boneName=thigh_r}}]
    constraintWeights [1]
   }
   n52 = "Transform (Modify) Bone - Bone: dyn_pouch_02_r"(ComponentPose=@n51, Alpha=1, Translation="0,0,1", Rotation="0,0,0", Scale="1,1,1") {
    translationMode=BMM_Additive
    translationSpace=BCS_BoneSpace
    alphaBoolBlend {blendOption=Linear}
    boneToModify {boneName=dyn_pouch_02_r}
    translation {z=1}
   }
   n53 = Component To Local(ComponentPose=@n52)
   n54 = Get bIsInFrontEnd()
   n55 = "Transform (Modify) Bone - Bone: dyn_necklace_02"(
    ComponentPose=Local To Component(LocalPose=Blend Poses by bool(BlendPose_0=@n53, BlendPose_1=@n49, bActiveValue=@n54) {blendType=Linear})
    Alpha=Get Neckale_alpha() -> (Neckale_alpha:double)
    Translation="0,0,0"
    Rotation="8,0,0"
    Scale="1,1,1"
   ) {
    rotationMode=BMM_Additive
    rotationSpace=BCS_BoneSpace
    alphaBoolBlend {blendOption=Linear}
    alphaScaleBiasClamp {bInterpResult=true, interpSpeedDecreasing=1}
    boneToModify {boneName=dyn_necklace_02}
    rotation {pitch=8}
   }
   n56 = "Transform (Modify) Bone - Bone: dyn_pouch_fr_01_l"(ComponentPose=@n55, Alpha=@n43, Translation="1.8,1,4", Rotation="0.000,20.000,-24", Scale="1,1,1") {
    rotationMode=BMM_Additive
    rotationSpace=BCS_BoneSpace
    scaleMode=BMM_Additive
    scaleSpace=BCS_BoneSpace
    translationMode=BMM_Additive
    translationSpace=BCS_BoneSpace
    alphaBoolBlend {blendOption=Linear}
    boneToModify {boneName=dyn_pouch_fr_01_l}
    rotation {roll=-24, yaw=20}
    translation=<1.8 1 4>
   }
   n57 = "Transform (Modify) Bone - Bone: dyn_pouch_fr_02_l"(ComponentPose=@n56, Alpha=@n43, Translation="-12.5,10,14", Rotation="30,32.571,-78", Scale="1,1,1") {
    rotationMode=BMM_Additive
    rotationSpace=BCS_BoneSpace
    scaleMode=BMM_Additive
    scaleSpace=BCS_BoneSpace
    translationMode=BMM_Additive
    translationSpace=BCS_BoneSpace
    alphaBoolBlend {blendOption=Linear}
    boneToModify {boneName=dyn_pouch_fr_02_l}
    rotation=<rot 30 32.571 -78>
    translation=<-12.5 10 14>
   }
   n58 = "Transform (Modify) Bone - Bone: dyn_pouch_fr_01_r"(ComponentPose=@n57, Alpha=@n43, Translation="1,1,-2", Rotation="-10.036,4.771,20.881", Scale="1,1,0.9") {
    rotationMode=BMM_Additive
    rotationSpace=BCS_BoneSpace
    scaleMode=BMM_Additive
    scaleSpace=BCS_BoneSpace
    translationMode=BMM_Additive
    translationSpace=BCS_BoneSpace
    alphaBoolBlend {blendOption=Linear}
    boneToModify {boneName=dyn_pouch_fr_01_r}
    rotation=<rot -10.036 4.771 20.8809>
    scale {z=0.9}
    translation=<1 1 -2>
   }
   n59 = "Transform (Modify) Bone - Bone: dyn_pouch_fr_02_r"(ComponentPose=@n58, Alpha=@n43, Translation="-0.000,-2,-2", Rotation="0,-0.001,-59.995", Scale="1,1,1") {
    rotationMode=BMM_Additive
    rotationSpace=BCS_BoneSpace
    scaleMode=BMM_Additive
    scaleSpace=BCS_BoneSpace
    translationMode=BMM_Additive
    translationSpace=BCS_BoneSpace
    alphaBoolBlend {blendOption=Linear}
    boneToModify {boneName=dyn_pouch_fr_02_r}
    rotation {roll=-59.995, yaw=-0.001}
    translation {y=-2, z=-2}
   }
   n60 = "Transform (Modify) Bone - Bone: dyn_strap_l"(ComponentPose=@n59, Alpha=@n43, Translation="0,0.5,0", Rotation="-30,0,0", Scale="1,1,1") {
    rotationMode=BMM_Additive
    rotationSpace=BCS_BoneSpace
    translationMode=BMM_Additive
    translationSpace=BCS_BoneSpace
    alphaBoolBlend {blendOption=Linear}
    boneToModify {boneName=dyn_strap_l}
    rotation {pitch=-30}
    translation {y=0.5}
   }
   n61 = "Transform (Modify) Bone - Bone: dyn_pouch_02_l"(
    ComponentPose=@n60
    Alpha=Get Pouch_l_correction() -> (Pouch_l_correction:double)
    Translation="-2,5,1"
    Rotation="-10,-7,-26"
    Scale="1,1,1"
   ) {
    rotationMode=BMM_Additive
    translationMode=BMM_Additive
    alphaBoolBlend {blendOption=Linear}
    boneToModify {boneName=dyn_pouch_02_l}
    rotation=<rot -10 -7 -26>
    translation=<-2 5 1>
   }
   n62 = "Transform (Modify) Bone - Bone: dyn_pouch_02_r"(
    ComponentPose=@n61
    Alpha=Get Pouch_R_Correction() -> (Pouch_R_Correction:double)
    Translation="0,-5,0"
    Rotation="19.684,10.629,-12.383"
    Scale="1,1,1"
   ) {
    rotationMode=BMM_Additive
    rotationSpace=BCS_BoneSpace
    translationMode=BMM_Additive
    translationSpace=BCS_BoneSpace
    alphaBoolBlend {blendOption=Linear}
    boneToModify {boneName=dyn_pouch_02_r}
    rotation=<rot 19.684 10.629 -12.3831>
    translation {y=-5}
   }
   n63 = Get Belt_r_correction() -> (Belt_r_correction:double)
   n64 = "Transform (Modify) Bone - Bone: dyn_pouch_fr_01_r"(ComponentPose=@n62, Alpha=@n63, Translation="0,2,2", Rotation="0,0,0", Scale="0.3,1,1") {
    rotationMode=BMM_Additive
    rotationSpace=BCS_BoneSpace
    scaleMode=BMM_Replace
    scaleSpace=BCS_BoneSpace
    translationMode=BMM_Additive
    alphaBoolBlend {blendOption=Linear}
    boneToModify {boneName=dyn_pouch_fr_01_r}
    scale {x=0.3}
    translation {y=2, z=2}
   }
   n65 = "Transform (Modify) Bone - Bone: dyn_pouch_fr_02_r"(ComponentPose=@n64, Alpha=@n63, Translation="-8,10,5", Rotation="0,0,-60", Scale="1,1,1") {
    rotationMode=BMM_Additive
    rotationSpace=BCS_BoneSpace
    scaleMode=BMM_Additive
    scaleSpace=BCS_BoneSpace
    translationMode=BMM_Additive
    alphaBoolBlend {blendOption=Linear}
    boneToModify {boneName=dyn_pouch_fr_02_r}
    rotation {roll=-60}
    translation=<-8 10 5>
   }
   n66 = "Transform (Modify) Bone - Bone: dyn_pouch_fr_01_l"(ComponentPose=@n65, Alpha=@n63, Translation="0,0,0", Rotation="0,0,0", Scale="1,1,1") {
    rotationMode=BMM_Additive
    rotationSpace=BCS_BoneSpace
    scaleMode=BMM_Additive
    scaleSpace=BCS_BoneSpace
    translationMode=BMM_Additive
    translationSpace=BCS_BoneSpace
    alphaBoolBlend {blendOption=Linear}
    boneToModify {boneName=dyn_pouch_fr_01_l}
   }
   n67 = "Transform (Modify) Bone - Bone: dyn_pouch_fr_02_l"(ComponentPose=@n66, Alpha=@n63, Translation="0,0,0", Rotation="0,0,0", Scale="1,1,1") {
    rotationMode=BMM_Additive
    rotationSpace=BCS_BoneSpace
    scaleMode=BMM_Additive
    scaleSpace=BCS_BoneSpace
    translationMode=BMM_Additive
    translationSpace=BCS_BoneSpace
    alphaBoolBlend {blendOption=Linear}
    boneToModify {boneName=dyn_pouch_fr_02_l}
   }
   n68 = "Transform (Modify) Bone - Bone: dyn_strap_r"(ComponentPose=@n67, Alpha=@n63, Translation="0,0,0", Rotation="0,0,0", Scale="1,1,1") {
    rotationMode=BMM_Additive
    rotationSpace=BCS_BoneSpace
    translationMode=BMM_Additive
    translationSpace=BCS_BoneSpace
    alphaBoolBlend {blendOption=Linear}
    boneToModify {boneName=dyn_strap_r}
   }
   n69 = Get BagCorrection() -> (BagCorrection:double)
   n70 = "Transform (Modify) Bone - Bone: dyn_pouch_01_r"(ComponentPose=@n68, Alpha=@n69, Translation="3,0,2", Rotation="-5,0,0", Scale="1,1,1") {
    rotationMode=BMM_Additive
    translationMode=BMM_Additive
    alphaBoolBlend {blendOption=Linear}
    boneToModify {boneName=dyn_pouch_01_r}
    rotation {pitch=-5}
    translation {x=3, z=2}
   }
   n71 = "Transform (Modify) Bone - Bone: dyn_pouch_01_l"(ComponentPose=@n70, Alpha=@n69, Translation="-2,0,2", Rotation="4,-10,-10", Scale="1,1,1") {
    rotationMode=BMM_Additive
    translationMode=BMM_Additive
    alphaBoolBlend {blendOption=Linear}
    boneToModify {boneName=dyn_pouch_01_l}
    rotation=<rot 4 -10 -10>
    translation {x=-2, z=2}
   }
   n72 = "Transform (Modify) Bone - Bone: dyn_pouch_bk_01_r"(ComponentPose=@n71, Alpha=1, Translation="0,0,0", Rotation="0.000,-10.000,0.000", Scale="1,1,1") {
    rotationMode=BMM_Additive
    rotationSpace=BCS_BoneSpace
    scaleMode=BMM_Additive
    scaleSpace=BCS_BoneSpace
    translationMode=BMM_Additive
    translationSpace=BCS_BoneSpace
    alphaBoolBlend {blendOption=Linear}
    boneToModify {boneName=dyn_pouch_bk_01_r}
    rotation {yaw=-10.0002}
   }
   n73 = "Transform (Modify) Bone - Bone: dyn_pouch_bk_01_l"(ComponentPose=@n72, Alpha=1, Translation="0,0,0", Rotation="0.000,-10.000,0.000", Scale="1,1,1") {
    rotationMode=BMM_Additive
    rotationSpace=BCS_BoneSpace
    scaleMode=BMM_Additive
    scaleSpace=BCS_BoneSpace
    translationMode=BMM_Additive
    translationSpace=BCS_BoneSpace
    alphaBoolBlend {blendOption=Linear}
    boneToModify {boneName=dyn_pouch_bk_01_l}
    rotation {yaw=-10.0003}
   }
   n74 = Get bIsCrouching()
   n75 = "Transform (Modify) Bone - Bone: dyn_pouch_03_l"!(ComponentPose=@n73, bAlphaBoolEnabled:bool=@n74, Translation="-4.085,0.211,0.000", Rotation="0,0,0", Scale="1,1,1") -> (Pose) {
    alphaInputType=Bool
    translationMode=BMM_Additive
    translationSpace=BCS_BoneSpace
    alphaBoolBlend {blendOption=Linear}
    boneToModify {boneName=dyn_pouch_03_l}
    translation=<-4.085 0.211>
   }
   n76 = "Transform (Modify) Bone - Bone: dyn_pouch_03_r"!(ComponentPose=@n75, bAlphaBoolEnabled:bool=@n74, Translation="3.772,0.589,-0.000", Rotation="0,0,0", Scale="1,1,1") -> (Pose) {
    alphaInputType=Bool
    translationMode=BMM_Additive
    translationSpace=BCS_BoneSpace
    alphaBoolBlend {blendOption=Linear}
    boneToModify {boneName=dyn_pouch_03_r}
    translation=<3.772 0.589>
   }
   Save cached pose 'SimBase'(Pose=Component To Local(ComponentPose=@n76)) {cacheName=SimBase}
   n77 = "Transform (Modify) Bone - Bone: neck_01_scale"!(
    ComponentPose=Local To Component(LocalPose=Copy Pose From Mesh() {bUseAttachedParent=true})
    bAlphaBoolEnabled:bool=True
    Translation="0,0,0"
    Scale="1,1.05,1.05"
   ) -> (Pose) {
    alphaInputType=Bool
    scaleMode=BMM_Additive
    scaleSpace=BCS_BoneSpace
    translationMode=BMM_Additive
    translationSpace=BCS_BoneSpace
    alphaBoolBlend {blendOption=Linear}
    boneToModify {boneName=neck_01_scale}
    scale {y=1.05, z=1.05}
   }
   n78 = "Transform (Modify) Bone - Bone: neck_02_scale"!(ComponentPose=@n77, bAlphaBoolEnabled:bool=True, Translation="0.5,0,0", Scale="1,1.05,1.05") -> (Pose) {
    alphaInputType=Bool
    scaleMode=BMM_Additive
    scaleSpace=BCS_BoneSpace
    translationMode=BMM_Additive
    translationSpace=BCS_BoneSpace
    alphaBoolBlend {blendOption=Linear}
    boneToModify {boneName=neck_02_scale}
    scale {y=1.05, z=1.05}
    translation {x=0.5}
   }
   n79 = "Transform (Modify) Bone - Bone: neck_01_scale"(ComponentPose=@n78, bAlphaBoolEnabled=True, Translation="2,0,0", Scale="1,1.1,1.1") {
    alphaInputType=Bool
    scaleMode=BMM_Additive
    scaleSpace=BCS_BoneSpace
    translationMode=BMM_Additive
    translationSpace=BCS_BoneSpace
    alphaBoolBlend {blendOption=Linear}
    boneToModify {boneName=neck_01_scale}
    scale {y=1.1, z=1.1}
    translation {x=2}
   }
   n80 = "Transform (Modify) Bone - Bone: neck_02_scale"(ComponentPose=@n79, bAlphaBoolEnabled=True, Translation="0.5,0,0", Scale="1,1.15,1.15") {
    alphaInputType=Bool
    scaleMode=BMM_Additive
    scaleSpace=BCS_BoneSpace
    translationMode=BMM_Additive
    translationSpace=BCS_BoneSpace
    alphaBoolBlend {blendOption=Linear}
    boneToModify {boneName=neck_02_scale}
    scale {y=1.15, z=1.15}
    translation {x=0.5}
   }
   n81 = "Constraint - sim (PPP)"!(
    ComponentPose=@n80
    Alpha=1
    ConstraintWeights_0=Get SkydiveRootSpace() -> (SkydiveRootSpace:float)
    ConstraintWeights_1:float=Get SkydiveSpineSpace() -> (SkydiveSpineSpace:float)
    ConstraintWeights_2:float=0.8
   ) -> (Pose) {
    alphaBoolBlend {blendOption=Linear}
    boneToModify {boneName=sim}
    constraintSetup [{targetBone {boneName=root}}, {targetBone {boneName=spine_01}}, {targetBone {boneName=spine_02}}]
    constraintWeights [1,1,1]
   }
   n82 = Anim Dynamics - Bone: dyn_glove_01_r(ComponentPose=@n81, Alpha=1, ExternalForce="0,0,0") {
    angularDampingOverride=0.9999
    angularSpringConstant=150
    bAngularSpring=true
    bGravityOverrideInSimSpace=true
    bOverrideAngularDamping=true
    bUseGravityOverride=true
    simulationSpace=RootRelative
    alphaBoolBlend {blendOption=Linear}
    boundBone {boneName=dyn_glove_01_r}
    componentAppliedLinearAccClamp=<1500 1500 1500>
    componentLinearAccScale=<0.1 0.1 0.1>
    componentLinearVelScale=<0.7 0.7 0.7>
    gravityOverride {z=-980}
    physicsBodyDefinitions [{
      sphereCollisionRadius=0
      boundBone {boneName=dyn_glove_01_r}
      boxExtents=<6 6 6>
      constraintSetup {angularLimitsMax=<0 10 10>, angularLimitsMin=<0 -10 -10>, angularTarget=<0 0 0>}
      localJointOffset=<-0.1 0 0>
     }]
   }
   n83 = Anim Dynamics - Bone: dyn_glove_01_l(ComponentPose=@n82, Alpha=1, ExternalForce="0,0,0") {
    angularDampingOverride=0.9999
    angularSpringConstant=150
    bAngularSpring=true
    bGravityOverrideInSimSpace=true
    bOverrideAngularDamping=true
    bUseGravityOverride=true
    simulationSpace=RootRelative
    alphaBoolBlend {blendOption=Linear}
    boundBone {boneName=dyn_glove_01_l}
    componentAppliedLinearAccClamp=<1500 1500 1500>
    componentLinearAccScale=<0.1 0.1 0.1>
    componentLinearVelScale=<0.7 0.7 0.7>
    gravityOverride {z=-980}
    physicsBodyDefinitions [{
      sphereCollisionRadius=0
      boundBone {boneName=dyn_glove_01_l}
      boxExtents=<6 6 6>
      constraintSetup {angularLimitsMax=<0 10 10>, angularLimitsMin=<0 -10 -10>, angularTarget=<0 0 0>}
      localJointOffset=<0.1 0 0>
     }]
   }
   n84 = "Constraint - dyn_pouch_fr_01_MSTR (PP)"!(ComponentPose=@n83, Alpha=1, ConstraintWeights_0=0.1194, ConstraintWeights_1:float=0.8806) -> (Pose) {
    alphaBoolBlend {blendOption=Linear}
    boneToModify {boneName=dyn_pouch_fr_01_MSTR}
    constraintSetup [{targetBone {boneName=spine_03}}, {targetBone {boneName=spine_04}}]
    constraintWeights [1,1]
   }
   n85 = "Constraint - dyn_cannister_01_r_MSTR (PPPP)"!(
    ComponentPose=@n84
    Alpha=1
    ConstraintWeights_0=0.0131
    ConstraintWeights_1:float=0.0528
    ConstraintWeights_2:float=0.2252
    ConstraintWeights_3:float=0.709
   ) -> (Pose) {
    alphaBoolBlend {blendOption=Linear}
    boneToModify {boneName=dyn_cannister_01_r_MSTR}
    constraintSetup [{targetBone {boneName=spine_01}}, {targetBone {boneName=spine_02}}, {targetBone {boneName=spine_03}}, {targetBone {boneName=spine_04}}]
    constraintWeights [1,1,1,1]
   }
   n86 = "Constraint - dyn_cannister_01_l_MSTR (PPPP)"!(
    ComponentPose=@n85
    Alpha=1
    ConstraintWeights_0=0.0131
    ConstraintWeights_1:float=0.0528
    ConstraintWeights_2:float=0.2251
    ConstraintWeights_3:float=0.709
   ) -> (Pose) {
    alphaBoolBlend {blendOption=Linear}
    boneToModify {boneName=dyn_cannister_01_l_MSTR}
    constraintSetup [{targetBone {boneName=spine_01}}, {targetBone {boneName=spine_02}}, {targetBone {boneName=spine_03}}, {targetBone {boneName=spine_04}}]
    constraintWeights [1,1,1,1]
   }
   n87 = "Constraint - dyn_pouch_01_l_MSTR (PP)"!(ComponentPose=@n86, Alpha=1, ConstraintWeights_0=0, ConstraintWeights_1:float=1) -> (Pose) {
    alphaBoolBlend {blendOption=Linear}
    boneToModify {boneName=dyn_pouch_01_l_MSTR}
    constraintSetup [{targetBone {boneName=thigh_l}}, {targetBone {boneName=pelvis}}]
    constraintWeights [1,1]
   }
   n88 = "Constraint - dyn_pouch_01_r_MSTR (PP)"!(ComponentPose=@n87, Alpha=1, ConstraintWeights_0=0, ConstraintWeights_1:float=1) -> (Pose) {
    alphaBoolBlend {blendOption=Linear}
    boneToModify {boneName=dyn_pouch_01_r_MSTR}
    constraintSetup [{targetBone {boneName=thigh_r}}, {targetBone {boneName=pelvis}}]
    constraintWeights [1,1]
   }
   n89 = "Constraint - dyn_pouch_01_l_MSTR (RR)"!(ComponentPose=@n88, Alpha=1, ConstraintWeights_0=0.5, ConstraintWeights_1:float=0.5) -> (Pose) {
    alphaBoolBlend {blendOption=Linear}
    boneToModify {boneName=dyn_pouch_01_l_MSTR}
    constraintSetup [{transformType=Rotation, targetBone {boneName=thigh_l}}, {transformType=Rotation, targetBone {boneName=pelvis}}]
    constraintWeights [1,1]
   }
   n90 = "Constraint - dyn_pouch_01_r_MSTR (RR)"!(ComponentPose=@n89, Alpha=1, ConstraintWeights_0=0.5, ConstraintWeights_1:float=0.5) -> (Pose) {
    alphaBoolBlend {blendOption=Linear}
    boneToModify {boneName=dyn_pouch_01_r_MSTR}
    constraintSetup [{transformType=Rotation, targetBone {boneName=thigh_r}}, {transformType=Rotation, targetBone {boneName=pelvis}}]
    constraintWeights [1,1]
   }
   n91 = Look At - Bone: dyn_pouch_bk_01_l(ComponentPose=@n90, Alpha=1) {
    alphaBoolBlend {blendOption=Linear}
    boneToModify {boneName=dyn_pouch_bk_01_l}
    lookAtLocation=<0 1>
    lookAtTarget {boneReference {boneName=dyn_pouch_bk_01_l_aim}}
    lookAt_Axis {axis=<1 0>}
    lookUp_Axis {axis {x=0, z=1}}
   }
   n92 = Look At - Bone: dyn_pouch_bk_01_r(ComponentPose=@n91, Alpha=1) {
    alphaBoolBlend {blendOption=Linear}
    boneToModify {boneName=dyn_pouch_bk_01_r}
    lookAtLocation=<0 1>
    lookAtTarget {boneReference {boneName=dyn_pouch_bk_01_r_aim}}
    lookAt_Axis {axis=<-1 0>}
    lookUp_Axis {axis {x=0, z=1}}
   }
   n93 = "Constraint - dyn_pouch_fr_01_r_MSTR (PP)"!(ComponentPose=@n92, Alpha=1, ConstraintWeights_0=0, ConstraintWeights_1:float=1) -> (Pose) {
    alphaBoolBlend {blendOption=Linear}
    boneToModify {boneName=dyn_pouch_fr_01_r_MSTR}
    constraintSetup [{targetBone {boneName=thigh_l}}, {targetBone {boneName=pelvis}}]
    constraintWeights [1,1]
   }
   n94 = "Constraint - dyn_pouch_fr_01_l_MSTR (PP)"!(ComponentPose=@n93, Alpha=1, ConstraintWeights_0=0, ConstraintWeights_1:float=1) -> (Pose) {
    alphaBoolBlend {blendOption=Linear}
    boneToModify {boneName=dyn_pouch_fr_01_l_MSTR}
    constraintSetup [{targetBone {boneName=thigh_r}}, {targetBone {boneName=pelvis}}]
    constraintWeights [1,1]
   }
   n95 = RigidBody!(ComponentPose=@n94, bAlphaBoolEnabled=Get bRigidBodyEnabled(), OverrideWorldGravity:Vector=Get FinalGravity(), ExternalForce) -> (Pose) {
    alphaInputType=Bool
    bClampLinearTranslationLimitToRefPose=true
    bOverrideWorldGravity=true
    overridePhysicsAsset=&/F_MED_Cavalry_Bandit_Necklace_Physics
    simulationSpace=BaseBoneSpace
    alphaBoolBlend {blendOption=Linear}
    baseBoneRef {boneName=spine_05}
    componentAppliedLinearAccClamp=<3000 3000 3000>
    componentLinearAccScale=<0.5 0.5 0.5>
    componentLinearVelScale=<0.7 0.7 0.7>
    overrideWorldGravity {z=-980}
   }
   n96 = Trail controller - Bone: dyn_necklace_04!(ComponentPose=@n95, bAlphaBoolEnabled:bool=NOT Boolean(A=Get bRigidBodyEnabled())) -> (Pose) {
    alphaInputType=Bool
    bLimitRotation=true
    bUsePlanarLimit=true
    chainLength=3
    relaxationSpeedScale=2.5
    planarLimits [{drivingBone {boneName=spine_04}, planeTransform {location=<0 17.5 0>, rotation=<rot -90 46 0>, scale=<1 1 1>}}]
    rotationLimits [{limitMax=<30 0 30>, limitMin=<-30 -30 -30>}, {limitMax=<30 0 30>, limitMin=<-30 -20 -30>}, {limitMax=<10 0 20>, limitMin=<-10 -10 -20>}]
    rotationOffsets [<0 0 0>, <0 0 0>, <0 0 0>]
    trailBone {boneName=dyn_necklace_04}
    trailRelaxationSpeed {editorCurveData {keys [<at 0 10 RCIM_Linear>,<at 1 4 RCIM_Linear>]}}
   }
   Save cached pose 'PreSimulate'(Pose=Component To Local(ComponentPose=@n96)) {cacheName=PreSimulate}
  ]}
 EventGraph {nodes [
   Event OnCharacterCustomizationLoaded()
   > n1 = Sequence()
   > n2 = Branch(Condition="integer >= integer"(A=Get CurrentLOD(), B=1))
   > Set LOD1NeckScaleNeeded(LOD1NeckScaleNeeded:bool=true) -> (Output_Get:bool)
   Set LOD1NeckScaleNeeded(execute=@n2.else, LOD1NeckScaleNeeded=false)
   n3 = Branch(execute=@n1.then_1, Condition="integer >= integer"(A=Get CurrentLOD(), B=2))
   > Set LOD2NeckScaleNeeded(LOD2NeckScaleNeeded:bool=true) -> (Output_Get:bool)
   Set LOD2NeckScaleNeeded(execute=@n3.else, LOD2NeckScaleNeeded=false)
   Event BlueprintInitializeAnimation()
   > Set CrouchRootSpace(CrouchRootSpace:float=1) -> (Output_Get:float)
   > Set CrouchSpineSpace(CrouchSpineSpace:float=0) -> (Output_Get:float)
   > Set SkeletalMesh(SkeletalMesh:Skeletal Mesh Component Object Reference=GetOwnerSkeletalMeshComponent()) -> (Output_Get:Skeletal Mesh Component Object Reference)
   Event OnEndCrouching()
   > Set CrouchRootSpace(CrouchRootSpace=0)
   > Set CrouchSpineSpace(CrouchSpineSpace=1)
   > Delay(Duration=0.2)
   > Set CrouchRootSpace(CrouchRootSpace=1)
   > Set CrouchSpineSpace(CrouchSpineSpace=0)
   > ResetDynamics(InTeleportType=TeleportPhysics)
   Event OnBeginCrouching()
   > Set CrouchRootSpace(CrouchRootSpace=0)
   > Set CrouchSpineSpace(CrouchSpineSpace=1)
   > ResetDynamics(InTeleportType=TeleportPhysics)
   > Delay(Duration=0.2)
   > Set CrouchRootSpace(CrouchRootSpace=1)
   > Set CrouchSpineSpace(CrouchSpineSpace=0)
   n4 = Event BlueprintUpdateAnimation()
   n5 = Get bIsCrouching()
   n6 = Get bIsMoving2D()
   > Set Crouch_walk(Crouch_walk:double=SelectFloat(A=1, B=0, bPickA=AND Boolean(A=@n5, B=@n6))) -> (Output_Get:double)
   n7 = Get PawnMovementDirectionAzimuth() -> (PawnMovementDirectionAzimuth:float)
   n8 = float - float(A=MapRangeClamped(Value=@n7, InRangeA=45, InRangeB=90), B=MapRangeClamped(Value=@n7, InRangeA=90, InRangeB=135)) {is=title K2Node_CallFunction}
   n9 = Get bIsCrouching()
   n10 = Get bIsMoving2D()
   > Set Right_Direct(Right_Direct:double=SelectFloat(A="Absolute (Float)"(A=@n8), B=0, bPickA=AND Boolean(A=@n9, B=@n10))) -> (Output_Get:double)
   n11 = Get PawnMovementDirectionAzimuth()
   n12 = float - float(A=MapRangeClamped(Value=@n11, InRangeA=-45, InRangeB=-90), B=MapRangeClamped(Value=@n11, InRangeA=-90, InRangeB=-135))
   n13 = Get bIsCrouching()
   n14 = Get bIsMoving2D()
   > Set Left_Direct(Left_Direct:double=SelectFloat(A="Absolute (Float)"(A=@n12), B=0, bPickA=AND Boolean(A=@n13, B=@n14))) -> (Output_Get:double)
   > Set FPS_Alpha(FPS_Alpha:double=MapRangeClamped(Value=@n4.DeltaTimeX, InRangeA=0.025, InRangeB=0.033)) -> (Output_Get:double)
   n15 = MapRangeClamped(Value=@n4.DeltaTimeX, InRangeA=0.016, InRangeB=0.033)
   > Set Component Linear Acc Scale(Component Linear Acc Scale:Vector="Lerp (Vector)"!(A:Vector="0.5,0.5,0.5", B:Vector="0.3,0.3,0.3", Alpha:float=@n15) -> (ReturnValue:Vector)) -> (Output_Get:Vector)
   > Set Component Linear Vel Scale(Component Linear Vel Scale:Vector="Lerp (Vector)"(A="0.7,0.7,0.7", B="0.3,0.3,0.3", Alpha=@n15)) -> (Output_Get:Vector)
   n16 = Get SkeletalMesh() -> (SkeletalMesh:Skeletal Mesh Component Object Reference)
   n17 = BreakVector(InVec=BreakTransform(InTransform=GetSocketTransform(self=@n16, InSocketName=pelvis, TransformSpace=RTS_Component)))
   > Set BagCorrection(BagCorrection:double=float * float(A=MapRangeClamped(Value=@n17.Z, InRangeA=74, InRangeB=45)) {is=title K2Node_PromotableOperator}) -> (Output_Get:double)
   n18 = Get SkeletalMesh()
   n19 = BreakRotator(InRot=BreakTransform(InTransform=GetDeltaTransformFromRefPose(self=@n18, BoneName=spine_05, BaseName=spine_03)).Rotation)
   n20 = MapRangeClamped(Value="Absolute (Float)"(A=@n19), InRangeA=0, InRangeB=5, OutRangeA=0.1, OutRangeB=0)
   n21 = MapRangeClamped(Value=@n19.Yaw, InRangeA=-16, InRangeB=-22)
   > Set Neckale_alpha(Neckale_alpha:double="Clamp (Float)"(Value=float + float(A=@n20, B=@n21), Min=0, Max=1)) -> (Output_Get:double)
   n22 = Get bIsPlayingEmote()
   > Set Pouch_02_r_correction(Pouch_02_r_correction:double="Clamp (Float)"(Value="To Float (Boolean)"(InBool=@n22), Min=0.3, Max=0.8)) -> (Output_Get:double)
   n23 = GetOwnerSkeletalMeshComponent()
   n24 = GetDeltaTransformFromRefPose(self=@n23, BoneName=thigh_l, BaseName=pelvis)
   n25 = MapRangeClamped(Value=BreakRotator(InRot=BreakTransform(InTransform=@n24).Rotation).Yaw, InRangeA=-28, InRangeB=-90)
   > Set Belt_l_correction(Belt_l_correction:double=@n25) -> (Output_Get:double)
   n26 = Get PawnVelocity()
   n27 = TryGetPawnOwner()
   n28 = "Absolute (Float)"(A=Dot Product!(
     A:Vector=Normalize!(A:Vector=@n26, Tolerance:float=0.0001) -> (ReturnValue:Vector) {is=title K2Node_CallFunction}
     B:Vector=GetActorForwardVector!(self:Actor Object Reference=@n27) -> (ReturnValue:Vector)
    ) -> (ReturnValue:double))
   > Set Pouch_l_correction(Pouch_l_correction:double=float * float(A=@n25, B=@n28)) -> (Output_Get:double)
   n29 = GetDeltaTransformFromRefPose(self=@n23, BoneName=thigh_r, BaseName=pelvis)
   n30 = MapRangeClamped(Value=BreakRotator(InRot=BreakTransform(InTransform=@n29).Rotation).Yaw, InRangeA=-28, InRangeB=-90)
   > Set Belt_r_correction(Belt_r_correction:double=Power!(Base:double=@n30, Exp:double=0.6) -> (ReturnValue:double) {is=title K2Node_CallFunction}) -> (Output_Get:double)
   > Set Pouch_R_Correction(Pouch_R_Correction:double=float * float(A=@n28, B=@n30)) -> (Output_Get:double)
   Event OnBeginSkydiving()
   > ResetDynamics(InTeleportType=TeleportPhysics)
   > ResetDynamics(InTeleportType=ResetPhysics)
   Event OnEndSkydiving()
   > ResetDynamics(InTeleportType=TeleportPhysics)
   > ResetDynamics(InTeleportType=ResetPhysics)
   GetOwnerSkeletalMeshComponent()
   Get bIsOnGround() -> (bIsOnGround:bool)
   Get SkeletalMesh()
  ]}
]
variables [
 BagCorrection {type=number}
 Belt_l_correction {type=number}
 Belt_r_correction {type=number}
 Component Linear Acc Scale {type=Vector, default=<0.5 0.5 0.5>}
 Component Linear Vel Scale {type=Vector, default=<0.7 0.7 0.7>}
 Crouch_walk {type=number}
 FPS_Alpha {type=number}
 Left_Direct {type=number}
 LOD1NeckScaleNeeded {type=boolean}
 LOD2NeckScaleNeeded {type=boolean}
 Neckale_alpha {type=number}
 Pouch_02_r_correction {type=number, default=0.9}
 Pouch_l_correction {type=number}
 Pouch_R_Correction {type=number}
 Right_Direct {type=number}
 SkeletalMesh {type=SkeletalMeshComponent}
]
