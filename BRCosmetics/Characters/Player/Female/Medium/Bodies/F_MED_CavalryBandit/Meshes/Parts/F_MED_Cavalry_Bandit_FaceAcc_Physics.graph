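# Physics asset for the female medium Cavalry Bandit face-accessory mesh, simulating its dynamic hair strands around a two-capsule head body.
# Strands are short chains of mostly capsule bodies with tight swing and twist limits; strand 14 starts with a box and forks into two branches.
# Some chains constrain directly to the head, while others hang from tiny _MSTR bodies whose head constraints leave swing and twist free.
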PhysicsAsset /BRCosmetics/Characters/Player/Female/Medium/Bodies/F_MED_CavalryBandit/Meshes/Parts/F_MED_Cavalry_Bandit_FaceAcc_Physics
bodies {
 dyn_hairstrand11_01 [{shapeType=Capsule, length=4, radius=0.8, shapeName=dyn_hairstrand11_01_capsule, center=<2 0 0>, rotation=<rot 90 0 0>}]
 dyn_hairstrand11_02 [{shapeType=Capsule, length=4, radius=0.8, shapeName=dyn_hairstrand11_02_capsule, center=<2 0 0>, rotation=<rot 90 0 0>}]
 dyn_hairstrand12_01 [{shapeType=Capsule, length=4.5, radius=0.6, shapeName=dyn_hairstrand12_01_capsule, center=<2 0 0>, rotation=<rot 90 0 0>}]
 dyn_hairstrand12_01_MSTR [{shapeType=Capsule, length=0.01, radius=0.01}]
 dyn_hairstrand13_01 [{
   shapeType=Capsule
   length=3.456
   radius=0.7012
   shapeName=dyn_hairstrand13_01_capsule
   center=<1.7974 0 0>
   rotation=<rot -90 -179.9998 179.9998>
  }]
 dyn_hairstrand13_01_MSTR [{shapeType=Capsule, length=0.01, radius=0.01}]
 dyn_hairstrand14_01 [{shapeType=Box, extentX=4.5, extentY=4, extentZ=0.5, center=<1.8761 0 0>}]
 dyn_hairstrand14_02 [{shapeType=Capsule, length=2.3895, radius=1.3605, shapeName=dyn_hairstrand14_02_capsule, center=<2.181 0 0>, rotation=<rot 90 0 0>}]
 dyn_hairstrand14_03 [{shapeType=Capsule, length=2.3895, radius=0.9855, shapeName=dyn_hairstrand14_03_capsule, center=<2 0 0>, rotation=<rot 90 0 0>}]
 dyn_hairstrand14_04 [{shapeType=Capsule, length=2.125, radius=1.625, shapeName=dyn_hairstrand14_04_capsule, center=<2.1089 0 0>, rotation=<rot 90 0 0>}]
 dyn_hairstrand14_05 [{
   shapeType=Capsule
   length=3.125
   radius=1.5
   shapeName=dyn_hairstrand14_05_capsule
   center=<2.7746 -1 0>
   rotation=<rot 90 -159.444 -139.4438>
  }]
 dyn_hairstrand15_01 [{shapeType=Capsule, length=1.8626, radius=1.6374, shapeName=dyn_hairstrand15_01_capsule, center=<2.2777 0 0>, rotation=<rot 90 0 0>}]
 dyn_hairstrand15_01_MSTR [{shapeType=Capsule, length=0.01, radius=0.01}]
 dyn_hairstrand15_02 [{shapeType=Capsule, length=1.3626, radius=1.6374, shapeName=dyn_hairstrand15_02_capsule, center=<2.3226 0 0>, rotation=<rot 84.0985 0 0>}]
 dyn_hairstrand1_01 [{shapeType=Capsule, length=4, radius=0.4, shapeName=dyn_hairstrand1_01_capsule, center=<2 0 0>, rotation=<rot 90 0 0>}]
 dyn_hairstrand1_02 [{shapeType=Capsule, length=0.275, radius=1.875, shapeName=dyn_hairstrand1_02_capsule, center=<1.4168 0 0>, rotation=<rot 90 0 0>}]
 dyn_hairstrand1_03 [{shapeType=Capsule, length=0.9, radius=2, shapeName=dyn_hairstrand1_03_capsule, center=<1.6077 1.5 0>, rotation=<rot 90 -165.9638 154.0363>}]
 dyn_hairstrand2_01 [{shapeType=Capsule, length=3, radius=2, shapeName=dyn_hairstrand2_01_capsule, center=<2.6705 0 1.5>, rotation=<rot 85 -179.9997 -179.9999>}]
 dyn_hairstrand2_01_MSTR [{shapeType=Capsule, length=0.01, radius=0.01}]
 dyn_hairstrand2_02 [{shapeType=Capsule, length=2.3928, radius=2.0901, shapeName=dyn_hairstrand2_02_capsule, center=<3.0953 0 1>, rotation=<rot 80.0003 0 -0.0005>}]
 dyn_hairstrand2_03 [{
   shapeType=Capsule
   length=1.3676
   radius=1.4986
   shapeName=dyn_hairstrand2_03_capsule
   center=<1.4944 0.5 0>
   rotation=<rot 80.0002 0 -25.0003>
  }]
 dyn_hairstrand3_01 [{shapeType=Capsule, length=5, radius=1, shapeName=dyn_hairstrand3_01_capsule, center=<1.8761 0 0>, rotation=<rot 90 0 0>}]
 dyn_hairstrand3_02 [{
   shapeType=Capsule
   length=1.875
   radius=1.75
   shapeName=dyn_hairstrand3_02_capsule
   center=<2.0319 0 0.5>
   rotation=<rot 85.0001 -0.0001 -0.0004>
  }]
 dyn_hairstrand3_03 [{shapeType=Capsule, length=5.125, radius=2.375, shapeName=dyn_hairstrand3_03_capsule, center=<4.2377 3 0>, rotation=<rot 85 30.0004 0>}]
 dyn_hairstrand4_01 [{shapeType=Capsule, length=4, radius=0.55, shapeName=dyn_hairstrand4_01_capsule, center=<2 0 0>, rotation=<rot 90 0 0>}]
 dyn_hairstrand4_01_MSTR [{shapeType=Capsule, length=0.01, radius=0.01}]
 dyn_hairstrand8_01 [{shapeType=Capsule, length=0.375, radius=1.35, shapeName=dyn_hairstrand8_01_capsule, center=<1.3705 0 0>, rotation=<rot 90 0 0>}]
 dyn_hairstrand8_01_MSTR [{shapeType=Capsule, length=0.01, radius=0.01}]
 dyn_hairstrand8_02 [{shapeType=Capsule, length=0.125, radius=1.6, shapeName=dyn_hairstrand8_02_capsule, center=<1.3705 0 0>, rotation=<rot 90 0 0>}]
 dyn_hairstrand9_01 [{shapeType=Capsule, length=4, radius=0.6, shapeName=dyn_hairstrand9_01_capsule, center=<2 0 0>, rotation=<rot 90 0 0>}]
 dyn_hairstrand9_02 [{shapeType=Capsule, length=4, radius=0.6, shapeName=dyn_hairstrand9_02_capsule, center=<2 0 0>, rotation=<rot 90 0 0>}]
 head [
  {shapeType=Capsule, length=4.5, radius=6.375, center=<3.6436 1.4768 0>, rotation=<rot -90 -180 -180>}
  {shapeType=Capsule, length=1.875, radius=5.25, center=<-1.4081 4.6109 0>, rotation=<rot 90 0 -10.0003>}
 ]
}
constraints [
 {bone1Name=dyn_hairstrand14_01, bone2Name=head, swing1LimitDegrees=45, swing2LimitDegrees=5, twistLimitDegrees=5}
 {bone1Name=dyn_hairstrand14_02, bone2Name=dyn_hairstrand14_01, swing1LimitDegrees=10, swing2LimitDegrees=25, twistLimitDegrees=5}
 {bone1Name=dyn_hairstrand14_03, bone2Name=dyn_hairstrand14_02, swing1LimitDegrees=10, swing2LimitDegrees=25, twistLimitDegrees=5}
 {bone1Name=dyn_hairstrand14_04, bone2Name=dyn_hairstrand14_01, swing1LimitDegrees=10, swing2LimitDegrees=25, twistLimitDegrees=5}
 {bone1Name=dyn_hairstrand14_05, bone2Name=dyn_hairstrand14_04, swing1LimitDegrees=10, swing2LimitDegrees=25, twistLimitDegrees=5}
 {bone1Name=dyn_hairstrand15_01, bone2Name=dyn_hairstrand15_01_MSTR, swing1LimitDegrees=5, swing2LimitDegrees=5, twistLimitDegrees=1}
 {bone1Name=dyn_hairstrand15_02, bone2Name=dyn_hairstrand15_01, swing1LimitDegrees=15, swing2LimitDegrees=15, twistLimitDegrees=1}
 {bone1Name=dyn_hairstrand1_01, bone2Name=head, swing1LimitDegrees=15, swing2LimitDegrees=15, twistLimitDegrees=5}
 {bone1Name=dyn_hairstrand1_02, bone2Name=dyn_hairstrand1_01, swing1LimitDegrees=25, swing2LimitDegrees=25, twistLimitDegrees=5}
 {bone1Name=dyn_hairstrand1_03, bone2Name=dyn_hairstrand1_02, swing1LimitDegrees=25, swing2LimitDegrees=25, twistLimitDegrees=5}
 {bone1Name=dyn_hairstrand9_01, bone2Name=head, swing1LimitDegrees=25, swing2LimitDegrees=25, twistLimitDegrees=5}
 {bone1Name=dyn_hairstrand9_02, bone2Name=dyn_hairstrand9_01, swing1LimitDegrees=25, swing2LimitDegrees=25, twistLimitDegrees=5}
 {bone1Name=dyn_hairstrand8_01, bone2Name=dyn_hairstrand8_01_MSTR, swing1LimitDegrees=25, swing2LimitDegrees=25, twistLimitDegrees=5}
 {bone1Name=dyn_hairstrand8_02, bone2Name=dyn_hairstrand8_01, swing1LimitDegrees=25, swing2LimitDegrees=25, twistLimitDegrees=5}
 {bone1Name=dyn_hairstrand12_01, bone2Name=dyn_hairstrand12_01_MSTR, swing1LimitDegrees=25, swing2LimitDegrees=25, twistLimitDegrees=5}
 {bone1Name=dyn_hairstrand11_01, bone2Name=head, swing1LimitDegrees=25, swing2LimitDegrees=25, twistLimitDegrees=5}
 {bone1Name=dyn_hairstrand11_02, bone2Name=dyn_hairstrand11_01, swing1LimitDegrees=25, swing2LimitDegrees=25, twistLimitDegrees=5}
 {bone1Name=dyn_hairstrand13_01, bone2Name=dyn_hairstrand13_01_MSTR, swing1LimitDegrees=25, swing2LimitDegrees=25, twistLimitDegrees=5}
 {bone1Name=dyn_hairstrand4_01, bone2Name=dyn_hairstrand4_01_MSTR, swing1LimitDegrees=25, swing2LimitDegrees=25, twistLimitDegrees=5}
 {bone1Name=dyn_hairstrand3_01, bone2Name=head, swing1LimitDegrees=15, swing2LimitDegrees=5, twistLimitDegrees=5}
 {bone1Name=dyn_hairstrand3_02, bone2Name=dyn_hairstrand3_01, swing1LimitDegrees=10, swing2LimitDegrees=15, twistLimitDegrees=5}
 {bone1Name=dyn_hairstrand3_03, bone2Name=dyn_hairstrand3_02, swing1LimitDegrees=10, swing2LimitDegrees=25, twistLimitDegrees=5}
 {bone1Name=dyn_hairstrand2_01, bone2Name=dyn_hairstrand2_01_MSTR, swing1LimitDegrees=5, swing2LimitDegrees=5, twistLimitDegrees=1}
 {bone1Name=dyn_hairstrand2_02, bone2Name=dyn_hairstrand2_01, swing1LimitDegrees=15, swing2LimitDegrees=15, twistLimitDegrees=1}
 {bone1Name=dyn_hairstrand2_03, bone2Name=dyn_hairstrand2_02, swing1LimitDegrees=15, swing2LimitDegrees=15, twistLimitDegrees=1}
 {
  bone1Name=dyn_hairstrand15_01_MSTR
  bone2Name=head
  swing1LimitDegrees=45
  swing1Motion=Free
  swing2LimitDegrees=45
  swing2Motion=Free
  twistLimitDegrees=45
  twistMotion=Free
 }
 {
  bone1Name=dyn_hairstrand8_01_MSTR
  bone2Name=head
  swing1LimitDegrees=45
  swing1Motion=Free
  swing2LimitDegrees=45
  swing2Motion=Free
  twistLimitDegrees=45
  twistMotion=Free
 }
 {
  bone1Name=dyn_hairstrand12_01_MSTR
  bone2Name=head
  swing1LimitDegrees=45
  swing1Motion=Free
  swing2LimitDegrees=45
  swing2Motion=Free
  twistLimitDegrees=45
  twistMotion=Free
 }
 {
  bone1Name=dyn_hairstrand13_01_MSTR
  bone2Name=head
  swing1LimitDegrees=45
  swing1Motion=Free
  swing2LimitDegrees=45
  swing2Motion=Free
  twistLimitDegrees=45
  twistMotion=Free
 }
 {
  bone1Name=dyn_hairstrand4_01_MSTR
  bone2Name=head
  swing1LimitDegrees=45
  swing1Motion=Free
  swing2LimitDegrees=45
  swing2Motion=Free
  twistLimitDegrees=45
  twistMotion=Free
 }
 {
  bone1Name=dyn_hairstrand2_01_MSTR
  bone2Name=head
  swing1LimitDegrees=45
  swing1Motion=Free
  swing2LimitDegrees=45
  swing2Motion=Free
  twistLimitDegrees=45
  twistMotion=Free
 }
]
