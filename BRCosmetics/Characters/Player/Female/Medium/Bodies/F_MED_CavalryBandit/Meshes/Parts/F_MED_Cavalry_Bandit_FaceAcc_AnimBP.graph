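# Animation Blueprint for the female Cavalry Bandit face accessory part, adding secondary hair motion on top of the pose copied from the parent mesh.
# Chained AnimDynamics nodes swing individual strands, then a root/head-weighted "sim" bone constraint feeds a neck-space RigidBody simulation with per-update gravity override.
# While targeting, tapering position constraints pull strand 14 toward the head.
# A skydiving-based windVector is computed each update but left unconnected in the AnimGraph.
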AnimBlueprint /BRCosmetics/Characters/Player/Female/Medium/Bodies/F_MED_CavalryBandit/Meshes/Parts/F_MED_Cavalry_Bandit_FaceAcc_AnimBP
target_skeleton=&/F_MED_Cavalry_Bandit_FaceAcc_Skeleton
graphs [
 AnimGraph {nodes [
   n1 = Anim Dynamics - Bone: dyn_hairstrand10_01(ComponentPose=Local To Component(LocalPose=Copy Pose From Mesh() {bUseAttachedParent=true}), Alpha=1, ExternalForce="0,0,0") {
    angularDampingOverride=1
    bGravityOverrideInSimSpace=true
    bOverrideAngularDamping=true
    gravityScale=0.35
    simulationSpace=RootRelative
    alphaBoolBlend {blendOption=Linear}
    boundBone {boneName=dyn_hairstrand10_01}
    componentLinearAccScale {z=0.5}
    componentLinearVelScale {z=0.5}
    physicsBodyDefinitions [{
      sphereCollisionRadius=0
      boundBone {boneName=dyn_hairstrand10_01}
      boxExtents=<5 10 10>
      constraintSetup {angularLimitsMax=<1 5 15>, angularLimitsMin=<-1 -15 -15>, angularTarget=<0 0 0>}
      localJointOffset=<2 0 0>
     }]
   }
   n2 = Anim Dynamics - Bone: dyn_hairstrand7_01(ComponentPose=@n1, Alpha=1, ExternalForce="0,0,0") {
    angularDampingOverride=1
    bGravityOverrideInSimSpace=true
    bOverrideAngularDamping=true
    gravityScale=0.35
    simulationSpace=RootRelative
    alphaBoolBlend {blendOption=Linear}
    boundBone {boneName=dyn_hairstrand7_01}
    componentLinearAccScale {z=0.5}
    componentLinearVelScale {z=0.5}
    physicsBodyDefinitions [{
      sphereCollisionRadius=0
      boundBone {boneName=dyn_hairstrand7_01}
      boxExtents=<5 10 10>
      constraintSetup {angularLimitsMax=<1 5 15>, angularLimitsMin=<-1 -15 -15>, angularTarget=<0 0 0>}
      localJointOffset=<2 0 0>
     }]
   }
   n3 = Anim Dynamics - Bone: dyn_hairstrand6_01(ComponentPose=@n2, Alpha=1, ExternalForce="0,0,0") {
    angularDampingOverride=1
    bGravityOverrideInSimSpace=true
    bOverrideAngularDamping=true
    gravityScale=0.35
    simulationSpace=RootRelative
    alphaBoolBlend {blendOption=Linear}
    boundBone {boneName=dyn_hairstrand6_01}
    componentLinearAccScale {z=0.5}
    componentLinearVelScale {z=0.5}
    physicsBodyDefinitions [{
      sphereCollisionRadius=0
      boundBone {boneName=dyn_hairstrand6_01}
      boxExtents=<5 10 10>
      constraintSetup {angularLimitsMax=<1 5 15>, angularLimitsMin=<-1 -15 -15>, angularTarget=<0 0 0>}
      localJointOffset=<2 0 0>
     }]
   }
   n4 = Anim Dynamics - Bone: dyn_hairstrand5_01_MSTR(ComponentPose=@n3, Alpha=1, ExternalForce="0,0,0") {
    angularDampingOverride=1
    bGravityOverrideInSimSpace=true
    bOverrideAngularDamping=true
    gravityScale=0.35
    simulationSpace=RootRelative
    alphaBoolBlend {blendOption=Linear}
    boundBone {boneName=dyn_hairstrand5_01_MSTR}
    componentLinearAccScale {z=0.5}
    componentLinearVelScale {z=0.5}
    physicsBodyDefinitions [{
      sphereCollisionRadius=0
      boundBone {boneName=dyn_hairstrand5_01_MSTR}
      boxExtents=<5 10 10>
      constraintSetup {angularLimitsMax=<1 5 15>, angularLimitsMin=<-1 -15 -15>, angularTarget=<0 0 0>}
      localJointOffset=<2 0 0>
     }]
   }
   n5 = Anim Dynamics - Bone: dyn_hairstrand12_01(ComponentPose=@n4, Alpha=1, ExternalForce="0,0,0") {
    angularDampingOverride=1
    bGravityOverrideInSimSpace=true
    bOverrideAngularDamping=true
    gravityScale=0.35
    simulationSpace=RootRelative
    alphaBoolBlend {blendOption=Linear}
    boundBone {boneName=dyn_hairstrand12_01}
    componentLinearAccScale {z=0.5}
    componentLinearVelScale {z=0.5}
    physicsBodyDefinitions [{
      sphereCollisionRadius=0
      boundBone {boneName=dyn_hairstrand12_01}
      boxExtents=<5 10 10>
      constraintSetup {angularLimitsMax=<1 5 15>, angularLimitsMin=<-1 -15 -15>, angularTarget=<0 0 0>}
      localJointOffset=<2 0 0>
     }]
   }
   n6 = Anim Dynamics - Bone: dyn_hairstrand11_02(ComponentPose=@n5, Alpha=1, ExternalForce="0,0,0") {
    angularDampingOverride=1
    bGravityOverrideInSimSpace=true
    bOverrideAngularDamping=true
    gravityScale=0.35
    simulationSpace=RootRelative
    alphaBoolBlend {blendOption=Linear}
    boundBone {boneName=dyn_hairstrand11_02}
    componentLinearAccScale {z=0.5}
    componentLinearVelScale {z=0.5}
    physicsBodyDefinitions [{
      sphereCollisionRadius=0
      boundBone {boneName=dyn_hairstrand11_02}
      boxExtents=<5 10 10>
      constraintSetup {angularLimitsMax=<1 5 15>, angularLimitsMin=<-1 -15 -15>, angularTarget=<0 0 0>}
      localJointOffset=<2 0 0>
     }]
   }
   n7 = Anim Dynamics - Bone: dyn_hairstrand13_01(ComponentPose=@n6, Alpha=1, ExternalForce="0,0,0") {
    angularDampingOverride=1
    bGravityOverrideInSimSpace=true
    bOverrideAngularDamping=true
    gravityScale=0.35
    simulationSpace=RootRelative
    alphaBoolBlend {blendOption=Linear}
    boundBone {boneName=dyn_hairstrand13_01}
    componentLinearAccScale {z=0.5}
    componentLinearVelScale {z=0.5}
    physicsBodyDefinitions [{
      sphereCollisionRadius=0
      boundBone {boneName=dyn_hairstrand13_01}
      boxExtents=<5 10 10>
      constraintSetup {angularLimitsMax=<1 5 15>, angularLimitsMin=<-1 -15 -15>, angularTarget=<0 0 0>}
      localJointOffset=<2 0 0>
     }]
   }
   n8 = "Constraint - sim (PP)"!(ComponentPose=@n7, Alpha=1, ConstraintWeights_0=0.15, ConstraintWeights_1:float=0.85) -> (Pose) {
    alphaBoolBlend {blendOption=Linear}
    boneToModify {boneName=sim}
    constraintSetup [{targetBone {boneName=root}}, {targetBone {boneName=head}}]
    constraintWeights [1,1]
   }
   n9 = RigidBody!(ComponentPose=@n8, Alpha, OverrideWorldGravity:Vector=Get GravityOverride(), ExternalForce) -> (Pose) {
    bClampLinearTranslationLimitToRefPose=true
    bOverrideWorldGravity=true
    overridePhysicsAsset=&/F_MED_Cavalry_Bandit_FaceAcc_Physics
    simulationSpace=BaseBoneSpace
    alphaBoolBlend {blendOption=Linear}
    baseBoneRef {boneName=neck_01}
    componentAppliedLinearAccClamp=<1500 1500 1500>
    componentLinearAccScale=<0.3 0.3 0.3>
    componentLinearVelScale=<0.3 0.3 0.3>
    overrideWorldGravity {z=-980}
   }
   n10 = "Constraint - dyn_hairstrand14_01 (P)"(ComponentPose=@n9, Alpha=Get targetingAlpha_0() -> (targetingAlpha_0:double), ConstraintWeights_0=1) {
    alphaBoolBlend {blendOption=Linear}
    boneToModify {boneName=dyn_hairstrand14_01}
    constraintSetup [{targetBone {boneName=head}}]
    constraintWeights [1]
   }
   n11 = "Constraint - dyn_hairstrand14_02 (P)"(ComponentPose=@n10, Alpha=Get targetingAlpha_1() -> (targetingAlpha_1:double), ConstraintWeights_0=1) {
    alphaBoolBlend {blendOption=Linear}
    boneToModify {boneName=dyn_hairstrand14_02}
    constraintSetup [{targetBone {boneName=head}}]
    constraintWeights [1]
   }
   n12 = "Constraint - dyn_hairstrand14_03 (P)"(ComponentPose=@n11, Alpha=Get targetingAlpha_2() -> (targetingAlpha_2:double), ConstraintWeights_0=1) {
    alphaBoolBlend {blendOption=Linear}
    boneToModify {boneName=dyn_hairstrand14_03}
    constraintSetup [{targetBone {boneName=head}}]
    constraintWeights [1]
   }
   n13 = Get targetingAlpha_3() -> (targetingAlpha_3:double)
   n14 = "Constraint - dyn_hairstrand14_04 (P)"(ComponentPose=@n12, Alpha=@n13, ConstraintWeights_0=1) {
    alphaBoolBlend {blendOption=Linear}
    boneToModify {boneName=dyn_hairstrand14_04}
    constraintSetup [{targetBone {boneName=head}}]
    constraintWeights [1]
   }
   n15 = "Constraint - dyn_hairstrand14_05 (P)"(ComponentPose=@n14, Alpha=@n13, ConstraintWeights_0=1) {
    alphaBoolBlend {blendOption=Linear}
    boneToModify {boneName=dyn_hairstrand14_05}
    constraintSetup [{targetBone {boneName=head}}]
    constraintWeights [1]
   }
   Output Pose(Result=Component To Local(ComponentPose=@n15)) {layerGroup=None}
   Get windVector() -> (windVector:Vector)
  ]}
 EventGraph {nodes [
   Event BlueprintUpdateAnimation()
   > n1 = Sequence()
   n2 = Break Fort Anim Input Skydiving External Force(FortAnimInput_SkydivingExternalForce=Get SkydivingExternalForceInput()) -> (
    bUseSkydivingVectorForce=~
    SkydivingVectorForce=~
    HeadToPelvisDirection=~
    FloatingMultiplier=~
    FloatingAdditive=~
    DivingMultiplier=~
    DivingAdditive=~
    ParachutingMultiplier=~
    ParachutingAdditive=~
   )
   > Set windVector(windVector:Vector=SelectVector(A=vector * float(A=@n2.SkydivingVectorForce, B=0.15), B="0, 0,-980", bPickA=Get bIsSkyDiving())) -> (Output_Get:Vector)
   n3 = Gravity Override State Function DEPRECATED(
    execute=@n1.then_1
    CharacterPartAnimInstance=Self-Reference()
    GameTimeInSeconds=GetGameTimeInSeconds()
    GravityOverrideParams=Get Gravity Override Params()
   )
   > Set GravityOverride(GravityOverride=@n3)
   Event OnBeginTargeting() {is=title K2Node_Event}
   > Set targetingAlpha_0(targetingAlpha_0:double=0.4) -> (Output_Get:double)
   > Set targetingAlpha_1(targetingAlpha_1:double=0.3) -> (Output_Get:double)
   > Set targetingAlpha_2(targetingAlpha_2:double=0.2) -> (Output_Get:double)
   > Set targetingAlpha_3(targetingAlpha_3:double=0.1) -> (Output_Get:double)
   Event OnEndTargeting() {is=title K2Node_Event}
   > Set targetingAlpha_0(targetingAlpha_0=0)
   > Set targetingAlpha_1(targetingAlpha_1=0)
   > Set targetingAlpha_2(targetingAlpha_2=0)
   > Set targetingAlpha_3(targetingAlpha_3=0)
   Event BlueprintInitializeAnimation()
   > Set targetingAlpha_0(targetingAlpha_0=0)
   > Set targetingAlpha_1(targetingAlpha_1=0)
   > Set targetingAlpha_2(targetingAlpha_2=0)
   > Set targetingAlpha_3(targetingAlpha_3=0)
   TryGetPawnOwner()
  ]}
]
variables [
 Gravity Override Params {type=GravityOverrideParamsStruct}
 GravityOverride {type=Vector}
 targetingAlpha_0 {type=number}
 targetingAlpha_1 {type=number}
 targetingAlpha_2 {type=number}
 targetingAlpha_3 {type=number}
 windVector {type=Vector}
]
